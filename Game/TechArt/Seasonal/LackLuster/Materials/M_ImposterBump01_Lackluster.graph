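# Masked two-sided octahedral impostor material for distant trees, with a seasonal LackLuster chrome look.
# Camera direction selects three atlas frames, parallax-shifted by normal-map alpha and weight-blended on a camera-facing sprite with pixel depth offset.
# Shadow-pass switches and mobile single-frame overrides vary the path.
# Color variation, trunk masking, optional snow and vertex-color static trunk feed tree fresnel, then MF_LackLuster_Base and a chrome color and roughness override.
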Material /Game/TechArt/Seasonal/LackLuster/Materials/M_ImposterBump01_Lackluster
graph {
 properties {
  bTangentSpaceNormal=false
  bUseMaterialAttributes=true
  bUsedWithInstancedStaticMeshes=true
  bUsedWithNanite=true
  bUsedWithStaticLighting=true
  blendMode=BLEND_Masked
  ditheredLODTransition=true
  opacityMaskClipValue=0.5
  pixelDepthOffsetMode=PDOM_Legacy
  twoSided=true
  usedShadingModels=Default Lit
 }
 nodes [
  n1 = Constant3Vector()
  n2 = MaterialFunctionCall(Float3=Multiply(
    A=TransformPosition(None=@n1) {transformType=TRANSFORMPOSSOURCE_World}
    B=ScalarParameter() {defaultValue=0.0012, group=Apollo, parameterName=Color Variation Range WS, sortPriority=10}
   ) {constB=0.0012}) {materialFunction=&BreakOutFloat3Components}
  n3 = Add(A=Multiply(A=@n2.R) {constB=0.5}, B=Multiply(A=@n2.G) {constB=0.5})
  n4 = Frac(None=@n2.B)
  Add(A=Multiply(A=Subtract(A=@n3) {constB=0.5}) {constB=2}, B=Multiply(A=Subtract(A=@n4) {constB=0.5}) {constB=2})
  MaterialExpressionCameraPositionWS_4 = CameraPositionWS()
  n5 = VectorParameter() {parameterName=LightVector, defaultValue=<rgba -1 0 1 0>}
  MaterialExpressionMultiply_9 = Multiply(A=Multiply(A=Normalize(VectorInput=@n5.RGB)) {constB=-1}) {constB=500000}
  n6 = Custom!(Main=@MaterialExpressionCameraPositionWS_4, Shadow=@MaterialExpressionMultiply_9) {
   code="##if SHADOW_DEPTH_SHADER \r\n{\r\nreturn Shadow;\r\n}\r\n##endif\r\nreturn Main;"
   description=Shadow Pass Switch
   inputs [
    {
     inputName=Main
     input {expression=&self:MaterialExpressionCameraPositionWS_4}
    }
    {inputName=Shadow, input {expression=&self:MaterialExpressionMultiply_9}}
   ]
  }
  n7 = MaterialFunctionCall() {materialFunction=/Game/ImpostorBaker/MaterialFunctions/ImpostorInfo}
  n8 = Normalize(VectorInput=Transform(None=Subtract(A=@n6, B=@n7.Position)) {transformSourceType=TRANSFORMSOURCE_World, transformType=TRANSFORM_Tangent})
  n9 = MaterialFunctionCall(Float3=@n8) {materialFunction=&BreakOutFloat3Components}
  n10 = Normalize(VectorInput=MaterialFunctionCall(X=@n9.R, Y=@n9.R, Z=Max(A=@n9.B) {constB=0.001}) {materialFunction=&MakeFloat3}.Result)
  n11 = Clamp(None=ConstantBiasScale(None=MaterialFunctionCall!(Unit Vector=@n10) {materialFunction=/Game/ImpostorBaker/MaterialFunctions/UnitVectorToHemiOctahedra}.Hemi Octahedron))
  n12 = Subtract(A=ScalarParameter() {defaultValue=16, parameterName=FramesXY})
  n13 = Floor(None=Multiply(A=@n11, B=@n12))
  n14 = Multiply(A=TextureCoordinate(), B=@n12)
  n15 = Floor(None=@n14)
  n16 = Frac(None=@n14)
  n17 = MaterialFunctionCall!(UV=@n16) {materialFunction=/Game/ImpostorBaker/MaterialFunctions/TriangleInterpolator}
  n18 = LinearInterpolate(Alpha=@n17.Mask) {constA=1, constB=0}
  n19 = LinearInterpolate(Alpha=@n17.Mask)
  n20 = Multiply(A=@n11, B=Subtract(A=ScalarParameter() {defaultValue=16, parameterName=FramesXY}))
  n21 = VertexInterpolator(VS=@n20)
  n22 = MaterialFunctionCall(UV=Frac(None=@n21.PS)) {materialFunction=&TriangleInterpolator}
  n23 = Multiply(
   A=SphereMask(A=@n13, B=@n15, Radius=ScalarParameter() {defaultValue=0.03, parameterName=FrameDebugRadius})
   B=Multiply(A=LinearInterpolate(A=@n18, B=@n19, Alpha=@n22.Mask), B=@n22.Result)
  )
  n24 = TextureCoordinate()
  n25 = ScalarParameter() {defaultValue=0.01, parameterName=FrameDebugRadius2}
  n26 = Add(A=Add(A=Multiply(A=@n23) {constB=0.2}, B=SphereMask(A=@n24, B=@n11, Radius=@n25)), B=Constant())
  n27 = ComponentMask(None=Multiply(A=Multiply(A=@n26) {constB=2}, B=ScalarParameter() {defaultValue=2, parameterName=DebugBoost})) {b=true, g=true, r=true}
  n28 = Desaturation(None=@n27, Fraction=Constant() {r=1})
  n29 = Divide(A=Floor(None=@n21.PS), B=ScalarParameter() {defaultValue=16, parameterName=FramesXY})
  n30 = MaterialFunctionCall() {materialFunction=&ImpostorInfo}
  n31 = ScalarParameter() {defaultValue=1, parameterName=WPOScale}
  n32 = MaterialFunctionCall() {materialFunction=&ImpostorInfo}
  n33 = CameraPositionWS()
  n34 = DotProduct(
   A=Normalize(VectorInput=Subtract(A=@n32.Position, B=@n33))
   B=MaterialFunctionCall() {materialFunction=&CameraDirectionVector}.Result
  )
  n35 = MaterialFunctionCall!(
   Center Location=@n30.Position
   0-1 UVs=Multiply(A=TextureCoordinate() {uTiling=10, vTiling=10}, B=ScalarParameter() {defaultValue=16, parameterName=FramesXY})
   XY scale=Multiply(A=@n30.Scaled Size, B=LinearInterpolate(A=@n31, Alpha=@n34))
   OverridePivot
   "Normals (Optional)"
  ) {materialFunction=&Sprite}
  n36 = Add(A=WorldPosition() {worldPositionShaderOffset=WPT_ExcludeAllShaderOffsets}.XYZ, B=Multiply(A=@n35.World position offset))
  n37 = Normalize(VectorInput=Transform(None=Subtract(A=@n6, B=@n36)) {transformSourceType=TRANSFORMSOURCE_World, transformType=TRANSFORM_Tangent})
  n38 = Multiply(A=@n37)
  n39 = Floor(None=@n20)
  n40 = MaterialFunctionCall!(
   InVec=@n38
   X Vec
   Y Vec
   Z Vec=MaterialFunctionCall!(Frame=@n39, XY Frames=ScalarParameter() {defaultValue=16, parameterName=FramesXY}) {materialFunction=/Game/ImpostorBaker/MaterialFunctions/ImposterFrameTransform_Setup}.World Z
  ) {materialFunction=/Game/ImpostorBaker/MaterialFunctions/ImposterFrameTransform}
  n41 = Multiply(
   A=MaterialFunctionCall() {materialFunction=&ImpostorInfo}.Scaled Size
   B=ScalarParameter() {defaultValue=1, parameterName=TanScaleFix}
  )
  n42 = CameraPositionWS() {desc=Ray Origin}
  n43 = Transform(None=Subtract(A=@n36, B=@n42)) {transformSourceType=TRANSFORMSOURCE_World, transformType=TRANSFORM_Tangent}
  n44 = MaterialFunctionCall() {materialFunction=&ImpostorInfo}
  n45 = Transform(None=Subtract(A=@n42, B=@n44.Position)) {transformSourceType=TRANSFORMSOURCE_World, transformType=TRANSFORM_Tangent}
  n46 = VertexInterpolator(VS=MaterialFunctionCall!(
    Plane Normal Axis=@n40.World Z
    Plane X Axis=@n40.World Z
    Plane Y Axis=Multiply(A=@n40.World Y) {constB=-1}
    UV Scale=@n41
    Ray Dir Local=@n43
    Ray Origin Local=@n45
   ) {materialFunction=/Game/ImpostorBaker/MaterialFunctions/VirtualPlaneCoordinates_Imposter}.UVs)
  n47 = Add(A=@n29, B=Divide(A=Saturate(None=@n46.PS), B=ScalarParameter() {defaultValue=16, parameterName=FramesXY}))
  MaterialExpressionReroute_8 = Reroute(None=@n47)
  n48 = Subtract(B=TextureSampleParameter2D(UVs=@n47) {constMipValue=0, parameterName=Normal, samplerSource=SSM_Wrap_WorldGroupSettings, texture=&Black}.A) {constA=0.5, constB=0.5}
  n49 = VertexInterpolator(VS=Divide(A=ComponentMask(None=@n40.Local Z Vector) {g=true, r=true}, B=ScalarParameter() {defaultValue=16, parameterName=FramesXY}) {constB=16})
  n50 = Constant() {r=1}
  MaterialExpressionAdd_0 = Add(A=@MaterialExpressionReroute_8, B=Multiply(A=Multiply(A=@n48, B=@n49.PS), B=@n50))
  n51 = Custom(Main=@MaterialExpressionAdd_0, Shadow=@MaterialExpressionReroute_8) {
   code="##if SHADOW_DEPTH_SHADER \r\n{\r\nreturn Shadow;\r\n}\r\n##endif\r\nreturn Main;"
   description=Shadow Pass Switch
   outputType=CMOT_Float2
   inputs [
    {inputName=Main, input {expression=&self:MaterialExpressionAdd_0}}
    {inputName=Shadow, input {expression=&self:MaterialExpressionReroute_8}}
   ]
  }
  n52 = MaterialFunctionCall(UV=Frac(None=@n21.PS)) {materialFunction=&TriangleInterpolator}
  n53 = LinearInterpolate(A=Constant2Vector() {g=1}, B=Constant2Vector() {r=1}, Alpha=@n52.Mask)
  n54 = ScalarParameter() {defaultValue=16, parameterName=FramesXY}
  n55 = LinearInterpolate(
   A=Constant2Vector() {g=1}
   B=Constant2Vector() {r=1}
   Alpha=MaterialFunctionCall(UV=Frac(None=@n20)) {materialFunction=&TriangleInterpolator}.Mask
  )
  n56 = MaterialFunctionCall(
   InVec=@n38
   Z Vec=MaterialFunctionCall(Frame=Add(A=@n39, B=@n55), XY Frames=ScalarParameter() {defaultValue=16, parameterName=FramesXY}) {materialFunction=&ImposterFrameTransform_Setup}.World Z
  ) {materialFunction=&ImposterFrameTransform}
  n57 = VertexInterpolator(VS=MaterialFunctionCall(
    Plane Normal Axis=@n56.World Z
    Plane X Axis=@n56.World Z
    Plane Y Axis=Multiply(A=@n56.World Y) {constB=-1}
    UV Scale=@n41
    Ray Dir Local=@n43
    Ray Origin Local=@n45
   ) {materialFunction=&VirtualPlaneCoordinates_Imposter}.UVs)
  n58 = Add(A=Add(A=@n29, B=Divide(A=@n53, B=@n54)), B=Divide(A=Saturate(None=@n57.PS), B=ScalarParameter() {defaultValue=16, parameterName=FramesXY}))
  n59 = Subtract(B=TextureSampleParameter2D(UVs=@n58) {constMipValue=0, parameterName=Normal, samplerSource=SSM_Wrap_WorldGroupSettings, texture=&Black}.A) {constA=0.5, constB=0.5}
  n60 = VertexInterpolator(VS=Divide(A=ComponentMask(None=@n56.Local Z Vector) {g=true, r=true}, B=ScalarParameter() {defaultValue=16, parameterName=FramesXY}) {constB=16})
  n61 = Add(A=@n58, B=Multiply(A=Multiply(A=@n59, B=@n60.PS), B=@n50))
  n62 = Constant() {r=1}
  n63 = ScalarParameter() {defaultValue=16, parameterName=FramesXY}
  n64 = Constant2Vector() {g=1, r=1}
  n65 = MaterialFunctionCall(
   InVec=@n38
   Z Vec=MaterialFunctionCall(Frame=Add(A=@n39, B=@n64), XY Frames=ScalarParameter() {defaultValue=16, parameterName=FramesXY}) {materialFunction=&ImposterFrameTransform_Setup}.World Z
  ) {materialFunction=&ImposterFrameTransform}
  n66 = VertexInterpolator(VS=MaterialFunctionCall(
    Plane Normal Axis=@n65.World Z
    Plane X Axis=@n65.World Z
    Plane Y Axis=Multiply(A=@n65.World Y) {constB=-1}
    UV Scale=@n41
    Ray Dir Local=@n43
    Ray Origin Local=@n45
   ) {materialFunction=&VirtualPlaneCoordinates_Imposter}.UVs)
  n67 = Add(A=Add(A=@n29, B=Divide(A=@n62, B=@n63)), B=Divide(A=Saturate(None=@n66.PS), B=ScalarParameter() {defaultValue=16, parameterName=FramesXY}))
  n68 = Subtract(B=TextureSampleParameter2D(UVs=@n67) {constMipValue=0, parameterName=Normal, samplerSource=SSM_Wrap_WorldGroupSettings, texture=&Black}.A) {constA=0.5, constB=0.5}
  n69 = VertexInterpolator(VS=Divide(A=ComponentMask(None=@n65.Local Z Vector) {g=true, r=true}, B=ScalarParameter() {defaultValue=16, parameterName=FramesXY}) {constB=16})
  n70 = Add(A=@n67, B=Multiply(A=Multiply(A=@n68, B=@n69.PS), B=@n50))
  n71 = Multiply(A=@n52.Result)
  n72 = Multiply(A=TextureCoordinate() {uTiling=10, vTiling=10}, B=ScalarParameter() {defaultValue=0.6, parameterName=Mip UV Scale})
  n73 = MaterialFunctionCall!(
   Texture Object=TextureObjectParameter() {parameterName=BaseColor, texture=&Black}
   R UV=@n51
   G UV=@n61
   B UV=@n70
   Weights=@n71
   Derivative UVs=@n72
  ) {materialFunction=/Game/ImpostorBaker/MaterialFunctions/FrameBlendWeights_TextureObject}
  MaterialExpressionComponentMask_10 = ComponentMask(None=@n73.Result) {a=true}
  MaterialExpressionComponentMask_15 = ComponentMask(None=@n73.R Sample) {a=true}
  n74 = Custom(Main=@MaterialExpressionComponentMask_10, Shadow=@MaterialExpressionComponentMask_15) {
   code="##if SHADOW_DEPTH_SHADER \r\n{\r\nreturn Shadow;\r\n}\r\n##endif\r\nreturn Main;"
   description=Shadow Pass Switch
   inputs [
    {
     inputName=Main
     input {expression=&self:MaterialExpressionComponentMask_10}
    }
    {
     inputName=Shadow
     input {expression=&self:MaterialExpressionComponentMask_15}
    }
   ]
  }
  n75 = MaterialFunctionCall() {materialFunction=/Game/ImpostorBaker/MaterialFunctions/Impostor_SingleFrameVersion}
  n76 = MaterialFunctionCall!(Default=@n74, Mobile Override=@n75.BaseColor A) {materialFunction=/Game/ImpostorBaker/MaterialFunctions/OverrideMobile}
  n77 = ComponentMask(None=@n76.Result) {r=true}
  n78 = StaticBool()
  n79 = StaticSwitch(True=Add(A=Multiply(A=@n28) {constB=5}, B=@n77), False=@n77, Value=@n78)
  Add(A=@n79, B=ScalarParameter() {parameterName=OpacityAdd})
  n80 = AppendVector(A=ComponentMask(None=@n13) {r=true}, B=ComponentMask(None=@n15) {r=true})
  n81 = MaterialFunctionCall(UVs=@n16, Texture Scale=ScalarParameter() {defaultValue=1, parameterName=UVscale}) {materialFunction=&ScaleUVsByCenter}
  AppendVector(
   A=Multiply(
    A=ScalarParameter() {defaultValue=1, parameterName=G}
    B=MaterialFunctionCall!(Vector2=@n80, MaximumNumberOfDigits, UVs=@n81.UVs, DebugTextLocation RG_UpperRight BA_LowerLeft, Component Spacing) {materialFunction=/Engine/Functions/Engine_MaterialFunctions02/Utility/DebugFloat2Values}.ColorCodedOutput
   )
   B=Constant()
  )
  Custom!(InWorldVector) {
   code="\r\n##if USE_INSTANCING || IS_MESHPARTICLE_FACTORY\r\n\t\tfloat3 temp;\r\n\t\ttemp.x = length(TransformLocalVectorToWorld(Parameters, float3(1,0,0)));\r\n\t\ttemp.y = length(TransformLocalVectorToWorld(Parameters, float3(0,1,0)));\r\n\t\ttemp.z = length(TransformLocalVectorToWorld(Parameters, float3(0,0,1)));\r\n\t\treturn mul(InWorldVector, (MaterialFloat3x3)transpose(Parameters.InstanceLocalToWorld)) / (temp*temp);\r\n\t##else\r\n\t\treturn mul(InWorldVector, (MaterialFloat3x3)Primitive.WorldToLocal);\r\n\t##endif"
   description=World to Loc
   inputs [{inputName=InWorldVector, input {expression=null}}]
  }
  Custom() {
   code="\r\n##if USE_INSTANCING || IS_MESHPARTICLE_FACTORY\r\n\t\tfloat3 temp;\r\n\t\ttemp.x = length(TransformLocalVectorToWorld(Parameters, float3(1,0,0)));\r\n\t\ttemp.y = length(TransformLocalVectorToWorld(Parameters, float3(0,1,0)));\r\n\t\ttemp.z = length(TransformLocalVectorToWorld(Parameters, float3(0,0,1)));\r\n\t\treturn mul(InWorldVector, (MaterialFloat3x3)transpose(Parameters.InstanceLocalToWorld)) / (temp*temp);\r\n\t##else\r\n\t\treturn mul(InWorldVector, (MaterialFloat3x3)Primitive.WorldToLocal);\r\n\t##endif"
   description=World to Loc
   inputs [{inputName=InWorldVector, input {expression=null}}]
  }
  n82 = VectorParameter() {parameterName=ExtractColor, defaultValue=<rgba 0.0239 0.5 0.0255 1>}
  n83 = MaterialFunctionCall(Default=ComponentMask(None=@n73.Result) {b=true, g=true, r=true}, Mobile Override=@n75.BaseColor RGB) {materialFunction=&OverrideMobile}
  n84 = Multiply(A=@n83.Result, B=Constant() {r=1})
  n85 = Abs(None=DotProduct(A=Normalize(VectorInput=@n82.RGB), B=Normalize(VectorInput=@n84)))
  n86 = ScalarParameter() {defaultValue=0.9682, parameterName=TrunkMaskMin}
  n87 = Saturate(None=Divide(A=Subtract(A=@n85, B=@n86), B=ScalarParameter() {defaultValue=0, parameterName=TrunkMaskLen}))
  n88 = StaticSwitchParameter(True=@n87, False=Constant() {r=1}) {parameterName=ExtractTrunkMask}
  n89 = MaterialFunctionCall(Float3=WorldPosition() {desc=Imposter is tiny and matches almost actor location, worldPositionShaderOffset=WPT_ExcludeAllShaderOffsets}.XYZ) {materialFunction=&BreakOutFloat3Components}
  n90 = ScalarParameter() {defaultValue=13000, group=Color, parameterName=Color Variation Distance WS, sortPriority=10}
  n91 = ScalarParameter() {defaultValue=11000, group=Color, parameterName=Color Variation Distance WS2, sortPriority=10}
  n92 = Add(A=Sine(None=Divide(A=@n89.R, B=@n90)), B=Sine(None=Divide(A=@n89.G, B=@n91)))
  n93 = Divide(A=@n92)
  n94 = ComponentMask(None=@n93) {r=true}
  n95 = ScalarParameter() {defaultValue=0.5, group=Color, parameterName=ValueVariation}
  n96 = ScalarParameter() {bUseCustomPrimitiveData=true, defaultValue=0.5, group=Color, parameterName=ValueVariation_primdata, primitiveDataIndex=5}
  n97 = PerInstanceCustomData!(DefaultValue) {dataIndex=5, desc=Alpha}
  n98 = VertexInterpolator(VS=PerInstanceCustomData(DefaultValue=Constant()) {dataIndex=3, desc=Alpha})
  n99 = VectorParameter() {bUseCustomPrimitiveData=true, parameterName=Color1_Base_primdata, defaultValue=<rgba 0.2016 0.2195 0.0423 1>}
  n100 = LinearInterpolate(A=@n99.A, B=@n98.PS, Alpha=@n98.PS)
  n101 = StaticSwitchParameter(True=LinearInterpolate(A=@n95, B=LinearInterpolate(A=@n96, B=@n97, Alpha=@n98.PS), Alpha=@n100), False=@n95) {parameterName=UsePrimitiveDataSeasonColoring}
  n102 = ScalarParameter() {defaultValue=-0.03, group=Color, parameterName=Hue Variation}
  n103 = ScalarParameter() {bUseCustomPrimitiveData=true, defaultValue=-0.03, group=Color, parameterName=Hue Variation_primdata, primitiveDataIndex=4}
  n104 = PerInstanceCustomData() {dataIndex=4, desc=Alpha}
  n105 = StaticSwitchParameter(True=LinearInterpolate(A=@n102, B=LinearInterpolate(A=@n103, B=@n104, Alpha=@n98.PS), Alpha=@n100), False=@n102) {parameterName=UsePrimitiveDataSeasonColoring}
  n106 = VectorParameter() {group=Color, parameterName=Color2_Lit, defaultValue=<rgba 0.2016 0.2195 0.0423 1>}
  n107 = LinearInterpolate(A=@n99.RGB, B=VertexInterpolator(VS=PerInstanceCustomData3Vector!(DefaultValue) {desc=Color}).PS, Alpha=@n98.PS)
  n108 = Multiply(A=ScalarParameter() {defaultValue=1.5, group=Color, parameterName=PrimDataColorMultiply_Lit}, B=@n107)
  n109 = VectorParameter() {group=Color, parameterName=Color1_Base, defaultValue=<rgba 0.0887 0.1046 0.0152 1>}
  n110 = StaticSwitchParameter(True=LinearInterpolate(A=@n109.RGB, B=@n107, Alpha=@n100), False=@n109.RGB) {parameterName=UsePrimitiveDataSeasonColoring}
  n111 = CollectionParameter() {
   collection=&FortniteMaterialParameters
   parameterName=SunlightDirectionalVector
  }
  n112 = MaterialFunctionCall(
   Texture Object=TextureObjectParameter() {parameterName=Normal, texture=&Black}
   R UV=@MaterialExpressionAdd_0
   G UV=@n61
   B UV=@n70
   Weights=@n71
   Derivative UVs=@n72
  ) {materialFunction=&FrameBlendWeights_TextureObject}
  n113 = ComponentMask(None=@n112.Result) {b=true, g=true, r=true}
  n114 = Transform(None=ConstantBiasScale(None=MaterialFunctionCall(Default=@n113, Mobile Override=@n75.Normal RGB) {materialFunction=&OverrideMobile}.Result) {bias=-0.5, scale=2})
  n115 = Saturate(None=DotProduct(A=ComponentMask(None=@n111) {b=true, g=true, r=true}, B=@n114))
  n116 = Saturate(None=Subtract(A=@n115, B=Constant() {r=0.5}))
  n117 = LinearInterpolate(
   A=StaticSwitchParameter(True=LinearInterpolate(A=@n106.RGB, B=@n108, Alpha=@n100), False=@n106.RGB) {parameterName=UsePrimitiveDataSeasonColoring}
   B=@n110
   Alpha=@n116
  )
  n118 = VectorParameter() {group=Color, parameterName=Color3_Shadows, defaultValue=<rgba 0.0356 0.0612 0.0116 1>}
  n119 = ScalarParameter() {defaultValue=0.5, group=Color, parameterName=PrimDataColorMultiply_Shadows}
  n120 = StaticSwitchParameter(True=LinearInterpolate(A=@n118.RGB, B=Multiply(A=@n119, B=@n107), Alpha=@n100), False=@n118.RGB) {parameterName=UsePrimitiveDataSeasonColoring}
  n121 = Constant() {r=-0.5}
  n122 = OneMinus(None=Saturate(None=Subtract(A=@n115, B=@n121)))
  n123 = Multiply(
   A=Add(A=Multiply(A=@n94, B=@n101))
   B=MaterialFunctionCall(Hue Shift Percentage=Multiply(A=@n94, B=@n105), Texture=LinearInterpolate(A=@n117, B=@n120, Alpha=@n122)) {materialFunction=&HueShift}.Result
  )
  n124 = ShadingPathSwitch(Default=@n123, Deferred=@n123, Forward=@n110, Mobile=@n110)
  n125 = Add(A=Multiply(A=Multiply(A=@n124, B=@n88)) {constB=2})
  n126 = StaticSwitchParameter(True=Desaturation(None=@n84, Fraction=ScalarParameter() {defaultValue=0.2, parameterName=Desaturation}), False=@n84) {parameterName=UseDesaturation}
  n127 = ComponentMask(None=@n126) {b=true, g=true, r=true}
  n128 = StaticSwitchParameter(True=Multiply(A=Subtract(A=@n125, B=@n88), B=@n127), False=LinearInterpolate(A=@n127, B=@n124, Alpha=@n88)) {defaultValue=true, group=Apollo, parameterName=UseDiffuse, sortPriority=1}
  n129 = ScalarParameter() {group=Apollo, parameterName=Fake SSS Scale}
  n130 = Add(A=@n129, B=ScalarParameter() {group=Apollo, parameterName=Fake SSS Scale add Mobile})
  n131 = Multiply(A=@n128, B=Multiply(A=@n88, B=FeatureLevelSwitch(Default=@n129, ES2=@n130, ES3_1=@n130, SM4_REMOVED=@n129, SM5=@n129, SM6=@n129)))
  n132 = ScalarParameter() {group=Apollo, parameterName=Fake SSS Desaturation}
  n133 = ScalarParameter() {group=Apollo, parameterName=Fake SSS Desaturation Mobile}
  Desaturation(None=@n131, Fraction=FeatureLevelSwitch(Default=@n132, ES2=@n133, ES3_1=@n133, SM4_REMOVED=@n132, SM5=@n132, SM6=@n132))
  n134 = Constant() {r=0.5}
  n135 = Divide(A=Add(A=@n92) {constB=2}) {constB=4}
  n136 = StaticSwitchParameter(True=@n135, False=ScalarParameter() {defaultValue=0.5, group=Apollo, parameterName=Fixed Color Manual Offset, sortPriority=-9}) {group=Apollo, parameterName=UseRandomColor, sortPriority=-10}
  n137 = TextureObjectParameter() {group=Apollo, parameterName=ColorPalette, sortPriority=99, texture=/Game/Environments/Apollo/Foliage/ColorLUT/Apollo_Foliage_LUT_Green}
  n138 = TextureSample(UVs=AppendVector(A=FeatureLevelSwitch(Default=@n136, ES2=@n134, ES3_1=@n134, SM4_REMOVED=@n136), B=Constant() {r=0.375}), Tex=@n137) {samplerSource=SSM_Wrap_WorldGroupSettings, texture=&Apollo_Foliage_LUT_Green}
  n139 = AppendVector(A=@n136, B=Constant() {r=0.125})
  n140 = Constant() {r=0.625}
  n141 = LinearInterpolate(
   A=LinearInterpolate(
    A=@n138.RGB
    B=TextureSample(UVs=@n139, Tex=@n137) {samplerSource=SSM_Wrap_WorldGroupSettings, texture=&Apollo_Foliage_LUT_Green}.RGB
    Alpha=@n116
   )
   B=TextureSample(UVs=AppendVector(A=@n136, B=@n140), Tex=@n137) {samplerSource=SSM_Wrap_WorldGroupSettings, texture=&Apollo_Foliage_LUT_Green}.RGB
   Alpha=@n122
  )
  FeatureLevelSwitch(Default=@n141, ES2=@n138.RGB, ES3_1=@n138.RGB, SM4_REMOVED=@n141)
  n142 = Distance(A=CameraPositionWS(), B=MaterialFunctionCall() {materialFunction=&ImpostorInfo}.Position)
  n143 = Multiply(A=MaterialFunctionCall() {materialFunction=&ImpostorInfo}.Scaled Size) {constB=0.25}
  n144 = Saturate(None=Divide(A=Subtract(A=@n142, B=@n143), B=@n143)) {desc=Don't Push Verts into Camera}
  LinearInterpolate(A=@n8, B=@n37, Alpha=Multiply(A=ScalarParameter() {defaultValue=1, parameterName=InterpolatePosition}, B=@n144))
  LinearInterpolate(
   A=VectorParameter() {parameterName=Manual Weights, defaultValue=<rgba 0 0 1 1>}.RGB
   B=@n52.Result
   Alpha=ScalarParameter() {defaultValue=1, parameterName=Enable Weights}
  )
  n145 = CameraPositionWS()
  n146 = MaterialFunctionCall() {materialFunction=&ImpostorInfo}
  n147 = Add(A=@n35.World position offset, B=WorldPosition().XYZ)
  n148 = Subtract(A=CameraPositionWS(), B=@n147)
  n149 = Normalize(VectorInput=@n148) {desc=Vector to Geo}
  n150 = MaterialFunctionCall() {materialFunction=&CameraDirectionVector}
  n151 = Multiply(
   A=Subtract(
    A=Distance(A=@n145, B=@n146.Position)
    B=MaterialFunctionCall(Vector 3=@n148) {desc=Dist to Geo, materialFunction=/Engine/Functions/Engine_MaterialFunctions02/Utility/VectorLength}.V3 Length
   )
   B=Abs(None=DotProduct(A=@n149, B=@n150.Result))
  )
  n152 = Normalize(VectorInput=Subtract(A=CameraPositionWS(), B=@n147)) {desc=CameraVector}
  n153 = Multiply(A=MaterialFunctionCall() {materialFunction=&ImpostorInfo}.Scaled Size) {constB=0.5}
  n154 = Add(
   A=Multiply(A=Multiply(A=@n151, B=@n149), B=ScalarParameter() {defaultValue=-1, parameterName=Center Align})
   B=Multiply(A=Multiply(A=@n152, B=@n153), B=ScalarParameter() {defaultValue=1, parameterName=Camera Push})
  )
  n155 = DistanceFieldsRenderingSwitch!(No=Constant(), Yes=@n154)
  MaterialExpressionFeatureLevelSwitch_0 = FeatureLevelSwitch(Default=@n154, ES2=@n155, ES3_1=@n155, SM4_REMOVED=@n154, SM5=@n154, SM6=@n154)
  MaterialExpressionConstant_6 = Constant()
  n156 = Custom!(Main=@MaterialExpressionFeatureLevelSwitch_0, Shadow=@MaterialExpressionConstant_6) {
   code="##if SHADOW_DEPTH_SHADER \r\n{\r\nreturn Shadow;\r\n}\r\n##endif\r\nreturn Main;"
   description=Shadow Pass Switch
   inputs [
    {
     inputName=Main
     input {expression=&self:MaterialExpressionFeatureLevelSwitch_0}
    }
    {inputName=Shadow, input {expression=&self:MaterialExpressionConstant_6}}
   ]
  }
  n157 = Add(A=@n35.World position offset, B=@n156) {constB=0}
  LinearInterpolate(A=@n157, B=@n157, Alpha=@n144)
  LinearInterpolate(A=@n44.Position, B=@n36, Alpha=@n144)
  n158 = Multiply(A=TextureCoordinate(), B=Subtract(A=ScalarParameter() {defaultValue=16, parameterName=FramesXY}))
  Multiply(
   A=ScalarParameter() {defaultValue=1, parameterName=Debug Grid}
   B=MaterialFunctionCall!(UV=Frac(None=@n158), Line Width=ScalarParameter() {defaultValue=0.03, parameterName=grid line}) {materialFunction=/Game/ImpostorBaker/MaterialFunctions/TriangulatedWireframe}.Result
  )
  Multiply(A=@n157, B=ScalarParameter() {defaultValue=0.9503})
  n159 = ComponentMask(None=CollectionParameter() {parameterName=Direction}) {b=true, g=true, r=true}
  Multiply(A=@n159) {constB=-1}
  n160 = DotProduct(A=@n159, B=CameraVectorWS()) {desc=L dot camera}
  n161 = Power(
   Base=Max(A=Saturate(None=@n160), B=ScalarParameter() {defaultValue=0.3, parameterName=ScatterMin})
   Exp=ScalarParameter() {defaultValue=2, parameterName=fresnel}
  )
  n162 = OneMinus(None=ComponentMask(None=TextureCoordinate()) {g=true})
  n163 = VectorParameter() {parameterName=SubsurfaceColor, defaultValue=<rgba 1 1 1 5>}
  n164 = Subtract(
   A=OneMinus(None=@n76.Result)
   B=ScalarParameter() {defaultValue=0.1, desc=">> Start of inner edge for edge glow", parameterName=Scatter Mask Min}
  )
  n165 = ScalarParameter() {defaultValue=2, desc=">> Width of gradient for edge glow", parameterName=Scatter Mask Len}
  n166 = Saturate(None=Multiply(A=@n163.A, B=Divide(A=@n164, B=@n165)))
  n167 = VectorParameter() {parameterName=Color, defaultValue=<rgba 1 0.4475 0.0933 5>}
  n168 = Multiply(A=@n167.RGB, B=@n167.RGB)
  Multiply(A=@n83.Result, B=Multiply(A=Multiply(A=@n161, B=@n162), B=Multiply(A=@n166, B=@n168)))
  Normalize(VectorInput=VectorParameter() {parameterName=LightVector, defaultValue=<rgba -1 0 1 0>}.RGB)
  OneMinus(None=@n94)
  PerInstanceCustomData3Vector() {desc=Color}
  ScalarParameter() {defaultValue=1, parameterName=MipForBleed}
  ScalarParameter() {defaultValue=1, parameterName=MipForBleed}
  ScalarParameter() {defaultValue=1, parameterName=MipForBleed}
  ScalarParameter() {defaultValue=1, parameterName=EnableParallax}
  ScalarParameter() {defaultValue=0.4951, group=Apollo, parameterName=Shift_Gradient}
  ScalarParameter() {defaultValue=-0.3204, group=Apollo, parameterName=Shift_Gradient_2}
  n169 = StaticBool() {value=true}
  n170 = VectorParameter() {parameterName=Emissive Overlay, defaultValue=<rgba 1 1 1 1>}
  n171 = TextureSampleParameter2D(UVs=Multiply(A=TextureCoordinate(), B=ScalarParameter() {defaultValue=1, parameterName=EmissiveScale})) {parameterName=EmissiveOverlay, samplerSource=SSM_Wrap_WorldGroupSettings, texture=&Black}
  n172 = Constant()
  n173 = Subtract(B=Step(X=Custom!() {code="View.ViewToClip[3][3]", desc=Ignore pixel depth offset is camera is ortho, outputType=CMOT_Float1, inputs []}))
  n174 = MaterialFunctionCall(Default=ComponentMask(None=@n112.Result) {a=true}, Mobile Override=@n75.Normal A) {materialFunction=&OverrideMobile}
  n175 = ScalarParameter() {defaultValue=350, parameterName=Default Mesh Size}
  n176 = MaterialFunctionCall() {materialFunction=/Engine/Functions/Engine_MaterialFunctions02/WorldPositionOffset/ObjectScale}
  n177 = ScalarParameter() {defaultValue=1, parameterName=PDO}
  n178 = CameraVectorWS()
  n179 = Abs(None=DotProduct(
    A=Normalize(VectorInput=@n178)
    B=MaterialFunctionCall() {materialFunction=&CameraDirectionVector}.Result
   ))
  n180 = Multiply(
   A=OneMinus(None=ComponentMask(None=@n174.Result) {r=true})
   B=Multiply(A=Multiply(A=@n175, B=@n176.Scale X), B=Multiply(A=@n177, B=@n179)) {constB=0.5}
  )
  n181 = Constant()
  n182 = Constant()
  n183 = QualitySwitch(Default=FeatureLevelSwitch(Default=@n180, ES2=@n181, ES3_1=@n181, SM4_REMOVED=@n180), Low=@n182, Medium=@n182)
  n184 = ScalarParameter() {defaultValue=0.03, parameterName=Gcut}
  n185 = Divide(A=Subtract(A=ComponentMask(None=@n83.Result) {g=true}, B=@n184) {constB=0.3}, B=@n184)
  n186 = Multiply(A=Multiply(A=@n83.Result, B=Saturate(None=@n185)), B=@n163.RGB)
  n187 = Subtract(A=ComponentMask(None=CameraPositionWS()) {b=true}, B=ScalarParameter() {defaultValue=10000, parameterName=Camera Z Fadeout Loc})
  n188 = ScalarParameter() {defaultValue=35000, parameterName=Cam Z Fade Length}
  n189 = Multiply(A=VectorParameter() {parameterName=MobileZOffset, defaultValue=<rgba 0 0 550 0>}.RGB, B=Saturate(None=Divide(A=@n187, B=@n188)))
  n190 = MaterialFunctionCall(Default=@n157, Mobile Override=DistanceFieldsRenderingSwitch(No=Add(A=@n75.WPO, B=@n189), Yes=@n157)) {materialFunction=&OverrideMobile}
  n191 = VectorParameter() {parameterName=BaseColor Multiplier, defaultValue=<rgba 1 1 1 1>}
  n192 = Multiply(A=@n191.RGB, B=@n191.RGB)
  n193 = VectorParameter() {parameterName=BaseColor Multiplier_Highlights, defaultValue=<rgba 1 1 1 1>}
  n194 = Multiply(A=@n193.RGB, B=@n193.RGB)
  n195 = ComponentMask(None=@n114) {b=true}
  n196 = ScalarParameter() {defaultValue=0.7, parameterName=ColorMaskMin}
  n197 = Saturate(None=Divide(A=Subtract(A=@n195, B=@n196), B=ScalarParameter() {defaultValue=0.2, parameterName=ColorMaskLen}))
  n198 = Multiply(
   A=StaticSwitchParameter(True=LinearInterpolate(A=@n192, B=@n194, Alpha=@n197), False=@n192) {parameterName=UseTwoColors}
   B=ScalarParameter() {defaultValue=1, parameterName=BaseColorBrightness}
  )
  n199 = LinearInterpolate(A=VectorParameter() {parameterName=BaseColor Multiplier_Trunk, defaultValue=<rgba 1 1 1 1>}.RGB, B=@n198, Alpha=@n87)
  n200 = StaticSwitchParameter(
   True=ComponentMask(None=Add(A=@n128, B=@n131)) {b=true, g=true, r=true}
   False=StaticSwitchParameter(True=Multiply(A=@n199, B=@n126), False=Multiply(A=@n198, B=@n126)) {parameterName=ExtractTrunkMask}
  ) {group=Apollo, parameterName=UseApolloColoring, sortPriority=-100}
  n201 = Normalize(VectorInput=@n114)
  n202 = Constant3Vector() {constant=<rgba 0 0 1 0>}
  n203 = Add(
   A=MaterialFunctionCall(In=DotProduct(A=@n201, B=@n202), Contrast=Constant()) {materialFunction=&CheapContrast}.Result
   B=ScalarParameter() {defaultValue=-0.2, group=Snow, parameterName=SnowNormal_Bias}
  )
  n204 = StaticSwitchParameter(True=LinearInterpolate(A=@n200, B=Constant() {desc=Snow albedo, r=0.7}, Alpha=Saturate(None=@n203)) {constB=0.7}, False=@n200) {group=Snow, parameterName=Use_BRSnow}
  n205 = SetMaterialAttributes!(
   MaterialAttributes
   Base Color=@n204
   Metallic
   Specular=ScalarParameter() {defaultValue=2, parameterName=Specular}
   Roughness=ScalarParameter() {defaultValue=0.5, parameterName=Roughness}
   Emissive Color=StaticSwitchParameter(True=Multiply(A=@n170.RGB, B=@n171.RGB), False=StaticSwitch(True=@n27, False=@n172, Value=@n78)) {parameterName=UseSimpleEmissiveOverlay}
   Opacity=ScalarParameter() {defaultValue=0.1, parameterName=Opacity}
   Opacity Mask=@n79
   Normal=@n114
   World Position Offset=@n190.Result
   Subsurface Color=Add(A=@n186, B=Multiply(A=@n186, B=@n166))
   Pixel Depth Offset=Multiply(A=@n183, B=Abs(None=@n173))
  ) {attributeSetTypes [
    #1
    #2
    #3
    #4
    #5
    #6
    #7
    #8
    #9
    #10
    #11
   ]}
  n206 = GetMaterialAttributes(None=@n205) {attributeGetTypes [
    #1
    #7
    #9
    #11
   ]}
  n207 = VertexColor()
  n208 = SetMaterialAttributes!(
   MaterialAttributes=@n206.MaterialAttributes
   Base Color=LinearInterpolate(
    A=TextureSampleParameter2D() {parameterName=BaseColor, samplerSource=SSM_Wrap_WorldGroupSettings, texture=&Black}.RGB
    B=@n206.Base Color
    Alpha=@n207.R
   )
   Opacity Mask=LinearInterpolate(A=Constant() {r=1}, B=@n206.Opacity Mask, Alpha=@n207.R)
   World Position Offset=LinearInterpolate(A=Constant3Vector(), B=@n206.World Position Offset, Alpha=@n207.R)
   Pixel Depth Offset=LinearInterpolate(A=Constant(), B=@n206.Pixel Depth Offset, Alpha=@n207.R)
  ) {attributeSetTypes [
    #1
    #7
    #9
    #11
   ]}
  n209 = StaticSwitchParameter(True=@n208, False=@n205) {parameterName=Use_StaticTrunk}
  n210 = MaterialFunctionCall!(
   IsStaticSnow
   UseAlternateStaticSnowMask
   AlternateStaticSnowMask
   "UseRoad&&OcclusionMask"=StaticBool()
   IsForGrassMeshToPushDown
   AdditionalMaskToRemoveSnow
   IsFallingSnowFX
   UseMinimapCoords
   IncreaseSnowUsingSlope
   UseSmoothBorder=@n169
   LerpHighM=Constant() {r=1.5}
   TemporalHeightLerpContrast
   IsLobbyIsland
   UseBR-IslandMask
   IsTree=@n169
   IsForRenderTarget
   UseSparkle
   UseSnowNormals
   UseAdditionalMaskToRemoveSnow
   IsTangentSpaceInput
   UseVertexNormalsForBasicSnow
   UseTangentSnowNormals
   Material=@n209
  ) {materialFunction=/Game/Athena/Apollo/Environments/Landscape/Materials/Biome/UniqueBiomeMaterialFunctions/MF_Apollo_Snow_Progression}
  n211 = GetMaterialAttributes(None=MaterialFunctionCall!(
    In=StaticSwitchParameter(True=@n210.Material, False=@n209) {group=Snow, parameterName=Use_BRSnow-Progressive}
    Mask=Constant() {r=1}
    IsImpostor=StaticBool() {value=true}
   ) {materialFunction=/Game/Environments/Apollo/Foliage/MaterialFunctions/MF_TreeFresnel}.Result) {attributeGetTypes [
    #8
    #9
    #1
    #7
   ]}
  n212 = VectorParameter() {parameterName=ChromoImposterNormal, defaultValue=<rgba 0.5 0.5 1 1>}
  n213 = Multiply(A=@n211.Normal, B=@n212.RGB)
  n214 = GetMaterialAttributes(None=MaterialFunctionCall!(
    bBlendMaterialAttributes
    MA_In
    UVsOverride
    Override Normals=@n213
    BlendMaterialAttributesAlpha
    BaseColorOverride
    WorldPositionOverride=@n211.World Position Offset
    VertexNormalOverride=@n213
    SourceNormalsStrengthOverride=@n212.RGB
   ) {materialFunction=/Game/TechArt/Seasonal/LackLuster/MaterialFunctions/MF_LackLuster_Base}.MA Out) {attributeGetTypes [#1]}
  SetMaterialAttributes!(
   MaterialAttributes=@n214.MaterialAttributes
   World Position Offset=@n211.World Position Offset
   Opacity Mask=@n211.World Position Offset
   Roughness=ScalarParameter() {parameterName=Chrome_ImpostorRoghness}
   Base Color=LinearInterpolate(
    A=VectorParameter() {parameterName=ChormeImpostorColor, defaultValue=<rgba 1 1 1 1>}.RGB
    B=@n214.Base Color
    Alpha=ScalarParameter() {defaultValue=0.2, parameterName=ChormeImpostorBlend}
   )
   Normal=@n211.World Position Offset
  ) {attributeSetTypes [
    #9
    #7
    #4
    #1
    #8
   ]}
  n215 = Constant2Vector()
  n216 = VectorParameter() {parameterName=ReferencePoint, defaultValue=<rgba 65995.8828 46048.0703 0 1>}
  n217 = Multiply(A=Subtract(A=WorldPosition().XYZ, B=@n216.RGB), B=Constant3Vector() {constant=<rgba 1 1 0 0>})
  n218 = Subtract(A=VectorParameter() {parameterName=ReferencePoint_1, defaultValue=<rgba 88680.0391 35843.0195 0 1>}.RGB, B=@n216.RGB)
  n219 = Normalize(VectorInput=@n218)
  n220 = Constant() {r=-5000}
  n221 = If(
   A=Divide(
    A=DotProduct(A=@n217, B=@n219)
    B=MaterialFunctionCall(Vector 2=@n218) {materialFunction=&VectorLength}.V2 Length
   )
   B=Constant() {desc=Was originally CubePathThreshold}
   "A > B"=Constant()
   "A == B"=@n220
   "A < B"=@n220
  )
  StaticSwitchParameter(True=Add(A=@n190.Result, B=AppendVector(A=@n215, B=@n221)), False=@n190.Result) {parameterName=CubeDestroyPath}
  n222 = VertexColor()
  StaticSwitchParameter(
   True=LinearInterpolate(
    A=TextureSampleParameter2D() {parameterName=BaseColor, samplerSource=SSM_Wrap_WorldGroupSettings, texture=&Black}.RGB
    B=@n204
    Alpha=@n222.R
   )
   False=@n204
  ) {parameterName=Use_StaticTrunk}
  StaticSwitchParameter(True=LinearInterpolate(A=Constant3Vector(), B=@n190.Result, Alpha=@n222.R), False=@n190.Result) {parameterName=Use_StaticTrunk}
  StaticSwitchParameter(True=LinearInterpolate(A=Constant() {r=1}, B=@n79, Alpha=@n222.R), False=@n79) {parameterName=Use_StaticTrunk}
  StaticSwitchParameter(True=LinearInterpolate(A=Constant(), B=@n183, Alpha=@n222.R), False=@n183) {parameterName=Use_StaticTrunk}
  TextureCoordinate() {uTiling=10, vTiling=10}
  VertexInterpolator(VS=@n135)
  VertexInterpolator()
  VertexInterpolator(VS=@n93)
 ]
}
