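# Parent material for LackLuster AI skeletal meshes, with a chrome physical material and a shared base function fed pre-skinned position and normal.
# Emissive adds a fake directional highlight, from view-space normals and inverted Fresnel, plus a sine-pulsed pre-solid warning color.
# Switches add cropped emissive texture and vertex-color element hiding through World Position Offset.
# The crop's lower-right corner uses B for both axes, despite its BA parameter name.
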Material /Game/TechArt/Seasonal/LackLuster/Materials/M_LackLuster_AI_Parent
graph {
 properties {
  bOutputTranslucentVelocity=true
  bUseMaterialAttributes=true
  bUsedWithMeshDeformer=true
  bUsedWithSkeletalMesh=true
  physMaterial=/Game/TechArt/Seasonal/LackLuster/Physics/Chrome_PhysicalMaterial
  pixelDepthOffsetMode=PDOM_Legacy
  usedShadingModels=Default Lit
 }
 nodes [
  n1 = PreSkinnedPosition()
  n2 = PreSkinnedNormal()
  n3 = GetMaterialAttributes(None=MaterialFunctionCall!(
    bBlendMaterialAttributes
    MA_In
    UVsOverride
    Override Normals
    BlendMaterialAttributesAlpha
    BaseColorOverride
    WorldPositionOverride=VertexInterpolator(VS=@n1).PS
    VertexNormalOverride=VertexInterpolator(VS=@n2).PS
    SourceNormalsStrengthOverride
   ) {materialFunction=/Game/TechArt/Seasonal/LackLuster/MaterialFunctions/MF_LackLuster_Base}.MA Out) {attributeGetTypes [#1, #2, #3]}
  n4 = MaterialFunctionCall!(
   Thickness=ScalarParameter() {defaultValue=0.5, parameterName=FresnelThickness}
   NormalWS=Transform(None=@n3.Normal)
   PowerExponent=ScalarParameter() {defaultValue=2, parameterName=FresnelExponent}
  ) {materialFunction=/Game/Weapons/Prototype/ObjectMover/Materials/MF_NormalizedFresnel}
  n5 = Transform(None=PixelNormalWS()) {transformSourceType=TRANSFORMSOURCE_World, transformType=TRANSFORM_View}
  n6 = Normalize(VectorInput=VectorParameter() {group=Fake Highlight, parameterName=Fresnel Highlight Vector, defaultValue=<rgba 0.3 0.6 0 0>}.RGB)
  n7 = Power(
   Base=Saturate(None=DotProduct(A=@n5, B=@n6))
   Exp=ScalarParameter() {defaultValue=1.7, group=Fake Highlight, parameterName=Fresnel Highlight Exponent}
  )
  n8 = Sine(None=Time()) {period=0.42}
  n9 = Add(
   A=Multiply(
    A=Multiply(A=@n4.PowerInverted, B=@n7)
    B=VectorParameter() {group=Fake Highlight, parameterName=Fresnel Highlight Color, defaultValue=<rgba 1.518 0.7778 2 0>}.RGB
   )
   B=Multiply(A=ConstantBiasScale(None=@n8), B=VectorParameter() {parameterName=PreSolidWarningColor}.RGB)
  )
  n10 = TextureObjectParameter() {group=Emissive, parameterName=Emissive, texture=&T_BlacksRGB}
  n11 = VectorParameter() {group=Emissive, parameterName=EmissiveUVs_RG_UpperLeftCorner_BA_LowerRightCorner, defaultValue=<rgba 0 0 1 1>}
  n12 = ComponentMask(None=@n11.RGB) {g=true, r=true}
  n13 = AppendVector(A=@n11.B, B=@n11.B)
  n14 = Add(
   A=@n9
   B=Multiply(
    A=MaterialFunctionCall!(UVs, TextureIn=@n10, UpperLeftCorner=@n12, LowerRightCorner=@n13, ExportFloat4) {materialFunction=/Engine/Functions/Engine_MaterialFunctions01/Texturing/TextureCropping}.CroppedMasked
    B=VectorParameter() {parameterName=CroppedEmissiveScale, defaultValue=<rgba 1 1 1 0>}.RGB
   )
  )
  SetMaterialAttributes!(
   MaterialAttributes=@n3.MaterialAttributes
   Emissive Color=Add(A=@n3.Emissive Color, B=StaticSwitchParameter(True=@n14, False=@n9) {group=Emissive, parameterName=UseCroppedEmissive})
   World Position Offset=StaticSwitchParameter(
    True=MaterialFunctionCall!(WPO=@n3.World Position Offset) {materialFunction=/Game/Packages/Fortress_SharedMaterials/Base_Material_Functions/Fortnite_HideElementsWithVertexColor}.Result
    False=@n3.World Position Offset
   ) {parameterName=UseVertexColorForMask}
  ) {attributeSetTypes [#1, #3]}
 ]
}
