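# Material function for the LackLuster seasonal chrome "blob" look, building a metallic surface that can optionally blend over incoming material attributes.
# Scrolling normals wrap cylindrically via vertex-normal angle and pre-skinned height, fading at the poles, with Fresnel varying their strength.
# Base color mixes curve-atlas reflections with time-of-day colors, hue-shifted by Fresnel and world noise.
# Static switches add phaseable radial ripples and emissive, camera-masked opacity and hit-effect WPO.
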MaterialFunction /Game/TechArt/Seasonal/LackLuster/MaterialFunctions/MF_LackLuster_Blob
graph {nodes [
  n1 = Multiply(
   A=ComponentMask(None=NamedRerouteUsage()) {g=true, r=true}
   B=ScalarParameter() {defaultValue=0.5, group=ReflectionCurve, parameterName=ReflectionCurveUVDistortionScale}
  )
  Add(A=@n1, B=NamedRerouteUsage())
  n2 = TextureSampleParameter2D(UVs=NamedRerouteUsage()) {
   parameterName=ChromeNormals
   samplerSource=SSM_Wrap_WorldGroupSettings
   samplerType=SAMPLERTYPE_Normal
   texture=/Game/TechArt/Seasonal/LackLuster/Textures/T_Lackluster_Swirl_N
  }
  n3 = TextureSampleParameter2D(UVs=NamedRerouteUsage()) {
   parameterName=ChromeNormals
   samplerSource=SSM_Wrap_WorldGroupSettings
   samplerType=SAMPLERTYPE_Normal
   texture=&T_Lackluster_Swirl_N
  }
  Add(A=@n2.RGB, B=@n3.RGB)
  n4 = NamedRerouteDeclaration(None=FunctionInput(Preview=SetMaterialAttributes!(MaterialAttributes)) {bUsePreviewValueAsDefault=true, inputName=MA_In, inputType=FunctionInput_MaterialAttributes}) {name=MA_In, nodeColor=<rgba 1 0.9844 0 1>}
  n5 = SetMaterialAttributes!(
   MaterialAttributes
   BaseColor=NamedRerouteUsage()
   Metallic=NamedRerouteUsage()
   Specular=ScalarParameter() {parameterName=Specular}
   Roughness=NamedRerouteUsage()
   OpacityMask=NamedRerouteUsage()
   Normal=NamedRerouteUsage()
   EmissiveColor=NamedRerouteUsage()
   WorldPositionOffset=NamedRerouteUsage()
   AmbientOcclusion=NamedRerouteUsage()
  ) {attributeSetTypes [
    #1
    #2
    #3
    #4
    #5
    #6
    #7
    #8
    #9
   ]}
  n6 = Max(
   A=NamedRerouteDeclaration(None=FunctionInput() {bUsePreviewValueAsDefault=true, inputName=BlendMaterialAttributesAlpha, inputType=FunctionInput_Scalar, sortPriority=1}) {name=BlendMaterialAttributesAlpha, nodeColor=<rgba 1 0 0.8203 1>}
   B=ScalarParameter() {parameterName=BlendMaterialAttributesAlphaOverride}
  )
  n7 = FunctionInput(Preview=StaticBool()) {bUsePreviewValueAsDefault=true, inputName=bBlendMaterialAttributes, inputType=FunctionInput_StaticBool, sortPriority=-1}
  FunctionOutput(None=StaticSwitch(
    True=BlendMaterialAttributes(A=@n4, B=@n5, Alpha=@n6)
    False=@n5
    Value=NamedRerouteDeclaration(None=@n7) {name=bBlendMaterialAttributes, nodeColor=<rgba 1 0 0.4453 1>}
   )) {outputName=MA Out}
  n8 = VectorParameter() {parameterName=MainColor, defaultValue=<rgba 1 0.7 0.4 1>}
  n9 = TextureSample(UVs=@n1) {
   samplerSource=SSM_Wrap_WorldGroupSettings
   samplerType=SAMPLERTYPE_LinearColor
   texture=/Game/TechArt/Seasonal/LackLuster/Textures/T_Lackluster_Triangles
  }
  n10 = CurveAtlasRowParameter(CurveTime=@n9.B) {
   atlas=/Game/TechArt/Seasonal/LackLuster/Curves/CA_Lackluster
   curve=/Game/TechArt/Seasonal/LackLuster/Curves/C_Lackluster_Blob
   defaultValue=11
   parameterName=ReflectionCurve
  }
  n11 = NamedRerouteUsage()
  n12 = Multiply(
   A=NamedRerouteDeclaration(None=Multiply(A=@n10, B=@n11)) {name=ReflectionCurve, nodeColor=<rgba 1 0.9609 0 1>}
   B=ScalarParameter() {defaultValue=5, parameterName=ReflectionCurveBaseColorScale}
  )
  n13 = ScalarParameter() {defaultValue=0.5, parameterName=ReflectionIntensity}
  LinearInterpolate(A=@n8.RGB, B=@n12, Alpha=@n13)
  LinearInterpolate(Alpha=NamedRerouteUsage())
  n14 = NamedRerouteUsage()
  n15 = ScalarParameter() {defaultValue=160, parameterName=NanomistTriangleUVDivisor}
  n16 = ComponentMask(None=NamedRerouteUsage()) {g=true, r=true}
  n17 = ScalarParameter() {defaultValue=0.05, parameterName=NanomistDistortionScale}
  n18 = TextureSample(UVs=Add(A=Divide(A=@n14, B=@n15) {constB=500}, B=Multiply(A=@n16, B=@n17))) {
   samplerSource=SSM_Wrap_WorldGroupSettings
   samplerType=SAMPLERTYPE_LinearColor
   texture=&T_Lackluster_Triangles
  }
  n19 = ReflectionVectorWS(CustomWorldNormal=Multiply(A=NamedRerouteUsage(), B=VectorParameter() {parameterName=PhaseReflectionVectorNormalStrength, defaultValue=<rgba 0.1 0.1 1 1>}.RGB))
  n20 = Multiply(A=Abs(None=ComponentMask(None=@n19) {b=true, r=true}), B=ScalarParameter() {defaultValue=0.3, parameterName=PhaseReflectionVectorScale})
  n21 = TextureSampleParameter2D(UVs=@n20) {
   group=Chrome_Phased
   parameterName=PhasedMap
   samplerSource=SSM_Wrap_WorldGroupSettings
   texture=/Game/TechArt/Seasonal/LackLuster/Textures/T_Lackluster_Phased
  }
  n22 = TextureSampleParameter2D(UVs=@n20) {
   group=Chrome_Phased
   parameterName=PhasedMap
   samplerSource=SSM_Wrap_WorldGroupSettings
   texture=&T_Lackluster_Phased
  }
  n23 = Max(A=@n21.RGB, B=@n22.RGB)
  n24 = Multiply(
   A=@n18.R
   B=Multiply(A=ShadingPathSwitch(Default=@n23, Mobile=@n21.RGB), B=ScalarParameter() {defaultValue=0.08, parameterName=UnphasedColorScale})
  )
  NamedRerouteDeclaration(None=StaticSwitchParameter(True=Multiply(A=@n18.G, B=@n24), False=Constant()) {parameterName=IsPhaseable}) {name=Phaseable/Phased Emissive, nodeColor=<rgba 1 0 0.2578 1>}
  n25 = MaterialFunctionCall(Float3=NamedRerouteUsage()) {materialFunction=&BreakOutFloat3Components}
  n26 = Constant3Vector() {constant=<rgba 0 0 1 0>}
  n27 = Power(Base=Abs(None=DotProduct(A=@n25.B, B=@n26))) {constExponent=3}
  n28 = Constant3Vector() {constant=<rgba 1 0 0 0>}
  n29 = Normalize(VectorInput=VertexNormalWS())
  n30 = Constant3Vector() {constant=<rgba 0 1 0 0>}
  n31 = Divide(A=Arctangent2Fast(Y=DotProduct(A=@n28, B=@n29), X=DotProduct(A=@n29, B=@n30))) {constB=6.283}
  n32 = ScalarParameter() {defaultValue=2, parameterName=U Tiling}
  n33 = PreSkinnedPosition()
  n34 = Divide(A=ComponentMask(None=VertexInterpolator(VS=@n33).PS) {b=true}, B=Multiply(A=ObjectRadius()) {constB=6.283})
  n35 = Multiply(A=Add(A=OneMinus(None=@n34)) {constB=-1}, B=ScalarParameter() {defaultValue=3, parameterName=V Tiling}) {constB=3}
  n36 = Time()
  n37 = ScalarParameter() {defaultValue=0.095, parameterName=BlobPanningSpeed}
  n38 = OneMinus(None=Frac(None=Multiply(A=@n36, B=@n37) {constB=0.2}))
  n39 = TextureCoordinate()
  n40 = TextureSample!(
   UVs=AppendVector(A=Multiply(A=@n31, B=@n32) {constB=2}, B=Add(A=@n35, B=@n38))
   Tex
   "DDX(UVs)"=DDX(Value=@n39)
   "DDY(UVs)"=DDY(Value=@n39)
   Apply View MipBias
  ) {mipValueMode=TMVM_Derivative, samplerType=SAMPLERTYPE_Normal, texture=/Game/TechArt/Seasonal/LackLuster/Textures/T_Lackluster_Phased_Normal}
  n41 = ScalarParameter() {defaultValue=3, group=Normals, parameterName=NormalIntensity}
  n42 = ScalarParameter() {defaultValue=1, group=Normals, parameterName=RipplesAmount, sliderMax=1}
  n43 = AppendVector(A=Multiply(A=ComponentMask(None=@n40.RGB) {g=true, r=true}, B=Multiply(A=@n41, B=@n42)), B=Constant() {r=1})
  n44 = CrossProduct(A=@n29, B=Constant3Vector() {constant=<rgba 0 0 1 0>})
  n45 = CrossProduct(A=@n44, B=@n29)
  n46 = Constant3Vector()
  n47 = LinearInterpolate(
   A=Transform(None=MaterialFunctionCall(VectorToTransform=@n43, BasisX=@n44, BasisY=@n45, BasisZ=@n29, BasisPosition=@n46) {materialFunction=&Transform3x3Matrix}.Transformed Vector) {transformSourceType=TRANSFORMSOURCE_World, transformType=TRANSFORM_Tangent}
   B=Transform(None=VertexNormalWS()) {transformSourceType=TRANSFORMSOURCE_World, transformType=TRANSFORM_Tangent}
   Alpha=Saturate(None=Multiply(A=@n27))
  )
  n48 = Normalize(VectorInput=ShadingPathSwitch(Default=@n47, Mobile=@n47))
  n49 = Multiply(A=VectorParameter() {group=Normals, parameterName=ChromeNormalScaleOuter, defaultValue=<rgba 0.18 0.18 1 0>}.RGB, B=@n48)
  n50 = MaterialFunctionCall(
   Exponent=ScalarParameter() {defaultValue=2.5, group=Normals, parameterName=EdgeNormal_FresnelExponent}
   Vertex / Pixel Normal=NamedRerouteUsage()
  ) {materialFunction=&MF_Fresnel_FX}
  n51 = LinearInterpolate(Alpha=StaticSwitchParameter(True=@n50.Fresnel Inverted Double Sided, False=@n50.Fresnel Inverted Double Sided) {parameterName=UseDoubleSidedFresnel}) {constA=0.05}
  n52 = LinearInterpolate(
   A=Multiply(A=@n48, B=VectorParameter() {group=Normals, parameterName=ChromeNormalScaleInner, defaultValue=<rgba 0.05 0.05 1 0>}.RGB)
   B=@n49
   Alpha=SphereMask(
    A=NamedRerouteDeclaration(None=@n51) {name=EdgeNormalFresnelInverted, nodeColor=<rgba 0 1 0.5547 1>}
    B=Constant() {r=0.5}
    Radius=ScalarParameter() {defaultValue=0.35, group=Normals, parameterName=EdgeNormalFresnelRadius}
    Hardness=Constant()
   )
  )
  n53 = NamedRerouteUsage()
  n54 = NamedRerouteUsage()
  n55 = NamedRerouteDeclaration(None=StaticSwitchParameter(True=Multiply(A=@n53, B=@n54), False=Constant()) {defaultValue=true, parameterName=UseRippleNormalsForPhaseable}) {name=PhasedNormalsScaledByMask, nodeColor=<rgba 1 0.7969 0 1>}
  n56 = Add(A=NamedRerouteUsage(), B=StaticSwitchParameter(True=Add(A=@n52, B=@n55), False=@n52) {parameterName=IsPhaseable})
  NamedRerouteDeclaration(None=FunctionInput(Preview=Normalize(VectorInput=@n56)) {bUsePreviewValueAsDefault=true, inputName=Override Normals}) {name=Normals, nodeColor=<rgba 1 0 0.9141 1>}
  n57 = CollectionParameter() {collection=&FortniteMaterialParameters, parameterName=PlayerPosition}
  n58 = ComponentMask(None=@n57) {b=true, g=true, r=true}
  n59 = WorldPosition() {worldPositionShaderOffset=WPT_ExcludeAllShaderOffsets}
  n60 = ScalarParameter() {defaultValue=280, parameterName=PlayerPhaseableDistance}
  n61 = ScalarParameter() {defaultValue=0.3, parameterName=PlayerPhaseableHardness}
  n62 = CollectionParameter() {collection=&FortniteMaterialParameters, parameterName=PlayerDirection}
  n63 = Saturate(None=DotProduct(A=Multiply(A=@n62) {constB=-1}, B=VertexNormalWS()))
  NamedRerouteDeclaration(None=Multiply(A=SphereMask(A=@n58, B=@n59.XYZ, Radius=@n60, Hardness=@n61), B=Power(Base=@n63) {constExponent=3})) {name=PlayerPhaseableMask, nodeColor=<rgba 0.1016 0 1 1>}
  n64 = VectorParameter() {group=DiffuseSource, parameterName=DiffuseSource_ColorDodgeBase, defaultValue=<rgba 0.4601 0.4251 0.4002 0>}
  n65 = TextureSampleParameter2D!(UVs, Level, Apply View MipBias) {
   group=DiffuseSource
   mipValueMode=TMVM_MipLevel
   parameterName=Diffuse
   samplerSource=SSM_Wrap_WorldGroupSettings
   texture=/Game/Weapons/WeaponSkins/Wraps/Textures/BlackOut1/TBlackOut2_D
  }
  n66 = TextureSampleParameter2D() {
   group=DiffuseSource
   mipValueMode=TMVM_MipLevel
   parameterName=Diffuse_Texture_2
   samplerSource=SSM_Wrap_WorldGroupSettings
   texture=&TBlackOut2_D
  }
  n67 = TextureSampleParameter2D() {
   group=DiffuseSource
   mipValueMode=TMVM_MipLevel
   parameterName=Diffuse_Texture_3
   samplerSource=SSM_Wrap_WorldGroupSettings
   texture=&TBlackOut2_D
  }
  n68 = TextureSampleParameter2D() {
   group=DiffuseSource
   mipValueMode=TMVM_MipLevel
   parameterName=Diffuse_Texture_4
   samplerSource=SSM_Wrap_WorldGroupSettings
   texture=&TBlackOut2_D
  }
  n69 = StaticBool() {value=true}
  n70 = StaticSwitchParameter(
   True=MaterialFunctionCall!(transitionFactor, 1=@n65.RGB, 2=@n66.RGB, 3=@n67.RGB, 4=@n68.RGB, Use 2 Textures=@n69, Use 3 Textures=@n69, Use 4 Textures=@n69) {materialFunction=/Game/Athena/WinterArt/Materials/MF_LerpTexturesBasedOnInt_Float3_4Layers}.Result
   False=@n65.RGB
  ) {parameterName=Use4Layers}
  n71 = GetMaterialAttributes(None=NamedRerouteUsage()) {attributeGetTypes [#1]}
  n72 = NamedRerouteUsage()
  n73 = Desaturation(
   None=StaticSwitch(True=LinearInterpolate(A=@n70, B=@n71.BaseColor, Alpha=@n72), False=@n70, Value=NamedRerouteUsage())
   Fraction=ScalarParameter() {defaultValue=0.92, group=DiffuseSource, parameterName=DiffuseSourceDesaturation}
  )
  n74 = OneMinus(None=@n73)
  n75 = Add(
   A=StaticSwitchParameter(True=Divide(A=@n64.RGB, B=@n74), False=@n73) {defaultValue=true, group=DiffuseSource, parameterName=UseColorDodgeForDiffuse}
   B=ScalarParameter() {defaultValue=-0.5, group=DiffuseSource, parameterName=DiffuseSourceBias}
  )
  n76 = ScalarParameter() {defaultValue=1.6, group=DiffuseSource, parameterName=DiffuseSourceScale}
  NamedRerouteDeclaration(None=Saturate(None=Multiply(A=@n75, B=@n76))) {name=Diffuse, nodeColor=<rgba 1 0 0.1172 1>}
  NamedRerouteDeclaration(None=StaticSwitchParameter(
    True=MaterialFunctionCall!(
     WSDeformationCenter
     Code_DestructionAmount
     Code_UseUvForDamage
     Code_UseHitEffects
     Code_Emissive
     WSDamageRadius
     CrackDamageShader
     Diffuse
     ScalarModDamageCracks
     ScalarOffsetforDamageCracks
     ExposeLerpFunctionForInvertingTheHitDistanceCalc
     InvertDamageFalloff
     CrackWorldPosition
     WorldPosition
     DistanceCullFadeAnimationOffset
    ) {materialFunction=/Game/Packages/jl_tests/Fortnite_HitEffects}.Hit WPO
    False=Constant()
   ) {defaultValue=true, parameterName=UseHitEffects}) {name=WPO, nodeColor=<rgba 1 0 0.2812 1>}
  NamedRerouteDeclaration(None=Divide(A=NamedRerouteUsage(), B=ScalarParameter() {defaultValue=800, parameterName=PhaseMaskUV_Divisor})) {name=PhaseMaskUVs, nodeColor=<rgba 1 0.9141 0 1>}
  n77 = CameraPositionWS()
  n78 = WorldPosition()
  n79 = ScalarParameter() {defaultValue=250, parameterName=PhasingCameraMaskRadius}
  n80 = ScalarParameter() {defaultValue=0.8, parameterName=PhasingCameraMaskHardness}
  NamedRerouteDeclaration(None=OneMinus(None=SphereMask(A=@n77, B=@n78.XYZ, Radius=@n79, Hardness=@n80))) {name=PhasingCameraMask, nodeColor=<rgba 1 0 0.5391 1>}
  n81 = MaterialFunctionCall!(
   UseLocalSpace
   WorldPosition=FunctionInput(Preview=WorldPosition() {worldPositionShaderOffset=WPT_ExcludeAllShaderOffsets}.XYZ) {bUsePreviewValueAsDefault=true, inputName=WorldPositionOverride, sortPriority=3}
   NormalOverride=NamedRerouteUsage()
  ) {materialFunction=/Game/Environments/Apollo/Sets/Skyfire/Materials/MaterialFunctions/MF_WorldAlignedUVs_HardEdge}
  n82 = TextureCoordinate()
  NamedRerouteDeclaration(None=FunctionInput(Preview=StaticSwitchParameter(True=@n81.Result, False=@n82) {parameterName=UseCheapTriplanarUVs}) {bUsePreviewValueAsDefault=true, inputName=UVsOverride, inputType=FunctionInput_Vector2}) {name=UVs, nodeColor=<rgba 0.2891 0 1 1>}
  n83 = TextureCoordinate()
  NamedRerouteDeclaration(None=DDX(Value=@n83)) {name=TexCoord_DDX, nodeColor=<rgba 0.7109 1 0 1>}
  NamedRerouteDeclaration(None=DDY(Value=@n83)) {name=TexCoord_DDY, nodeColor=<rgba 1 0 0.3281 1>}
  n84 = CollectionParameter() {
   collection=&FortniteMaterialParameters
   parameterName=SunlightDirectionalVector
  }
  n85 = DotProduct(A=ReflectionVectorWS(CustomWorldNormal=PixelNormalWS()), B=ComponentMask(None=@n84) {b=true, g=true, r=true})
  NamedRerouteDeclaration(None=Power(Base=Saturate(None=@n85), Exp=ScalarParameter() {defaultValue=40, parameterName=ReflectionVectorExponent})) {name=SpecLightMask, nodeColor=<rgba 0 1 0.5312 1>}
  n86 = ScalarParameter() {defaultValue=1.7, parameterName=SunMoonFresnelBrightness}
  n87 = ScalarParameter() {defaultValue=2, parameterName=SunMoonFresnelExponent}
  n88 = CollectionParameter() {
   collection=&FortniteMaterialParameters
   parameterName=SunAndMoonModelDirectionalVector
  }
  n89 = ComponentMask(None=@n88) {b=true, g=true, r=true}
  n90 = PixelNormalWS()
  NamedRerouteDeclaration(None=LinearInterpolate(
    A=ScalarParameter() {defaultValue=0.45, parameterName=SunMoonMaskLow}
    B=ScalarParameter() {defaultValue=1, parameterName=SunMoonMaskHigh}
    Alpha=MaterialFunctionCall(Brightness=@n86, Exponent=@n87, Camera Vector=@n89, Vertex / Pixel Normal=@n90) {materialFunction=&MF_Fresnel_FX}.Fresnel
   )) {name=SunMoonMaskRemapped, nodeColor=<rgba 1 0 0.5156 1>}
  n91 = NamedRerouteUsage()
  n92 = TextureSampleParameter2D!(UVs, Apply View MipBias) {
   group=OpacityMask
   parameterName=OpacityMask
   samplerSource=SSM_Wrap_WorldGroupSettings
   texture=/Game/Creative/BuildingActors/Props/Textures/CP_White_Texutre
  }
  n93 = StaticSwitchParameter(
   True=Multiply(
    A=GetMaterialAttributes(None=@n91) {attributeGetTypes [#5]}.OpacityMask
    B=ChannelMaskParameter(None=@n92.RGBA) {parameterName=OpacityMaskChannel}.R
   )
   False=Constant() {r=1}
  ) {parameterName=UseOpacityMaskTexture}
  n94 = NamedRerouteUsage()
  NamedRerouteDeclaration(None=StaticSwitchParameter(True=Multiply(A=@n93, B=@n94), False=@n93) {parameterName=UsePhasingCameraMask}) {name=OpacityMask, nodeColor=<rgba 0.3828 1 0 1>}
  n95 = DotProduct(A=NamedRerouteUsage(), B=Constant3Vector() {constant=<rgba 0 0 -1 0>})
  NamedRerouteDeclaration(None=SphereMask(
    A=Saturate(None=@n95)
    B=ScalarParameter() {defaultValue=0.1, parameterName=VertexNormalMaskCenter}
    Radius=ScalarParameter() {defaultValue=0.2, parameterName=VertexNormalMaskRadius}
    Hardness=ScalarParameter() {defaultValue=0.5, parameterName=VertexNormalMaskHardness}
   )) {name=VertexNormalMask.-Z, nodeColor=<rgba 0 0.3203 1 1>}
  n96 = Power(Base=Fresnel(Normal=VertexNormalWS()) {exponent=1}, Exp=ScalarParameter() {defaultValue=5, parameterName=FresnelExponent}) {constExponent=5}
  n97 = ScalarParameter() {defaultValue=0.1, parameterName=FresnelIntensity}
  n98 = Multiply(
   A=MaterialFunctionCall!(
    Thickness=ScalarParameter() {defaultValue=1, group=Emissive, parameterName=NormalizedFresnelThickness}
    NormalWS=NamedRerouteUsage()
    PowerExponent=ScalarParameter() {defaultValue=1, group=Emissive, parameterName=NormalizedFresnelExponent}
   ) {materialFunction=/Game/Weapons/Prototype/ObjectMover/Materials/MF_NormalizedFresnel}.PowerInverted
   B=ScalarParameter() {defaultValue=1, group=Emissive, parameterName=NormalizedFresnelScale}
  )
  n99 = NamedRerouteUsage()
  n100 = TextureSampleParameter2D() {parameterName=Emissive, texture=&TBlackOut2_D}
  n101 = NamedRerouteUsage()
  n102 = NamedRerouteUsage()
  n103 = ScalarParameter() {defaultValue=0.3, parameterName=ReflectionCurveEmissiveScale}
  n104 = Add(
   A=StaticSwitchParameter(True=Add(A=@n100.RGB, B=@n101), False=@n101) {group=Emissive, parameterName=UseEmissiveTexture}
   B=Multiply(A=Multiply(A=@n102, B=@n103), B=NamedRerouteUsage())
  )
  n105 = StaticSwitchParameter(True=Add(A=Multiply(A=@n98, B=@n99), B=@n104), False=@n104) {group=Emissive, parameterName=UseEmissiveNormalizedFresnel}
  n106 = Add(A=Multiply(A=NamedRerouteUsage(), B=ScalarParameter() {defaultValue=1, parameterName=EmissiveBaseColorBoostScale}), B=@n105)
  NamedRerouteDeclaration(None=Add(A=Multiply(A=@n96, B=@n97), B=StaticSwitchParameter(True=@n106, False=@n105) {group=Emissive, parameterName=UseEmissiveBaseColorBoost})) {name=Emissive, nodeColor=<rgba 1 0 0.8203 1>}
  n107 = Multiply(A=ScalarParameter() {defaultValue=0.25, group=HueShift, parameterName=AnimatedFresnelHueShiftSpeed}, B=Time())
  n108 = NamedRerouteUsage()
  n109 = MaterialFunctionCall(
   Brightness=ScalarParameter() {defaultValue=1, group=HueShift, parameterName=FresnelHueShiftBrightness}
   Exponent=ScalarParameter() {defaultValue=1.5, group=HueShift, parameterName=FresnelHueShiftExponent}
   Vertex / Pixel Normal=NamedRerouteDeclaration(None=Transform(None=@n108)) {name=NormalsWorldSpace, nodeColor=<rgba 1 0.5625 0 1>}
  ) {materialFunction=&MF_Fresnel_FX}
  n110 = StaticSwitchParameter(True=@n109.Fresnel Double Sided, False=@n109.Fresnel Double Sided) {parameterName=UseDoubleSidedFresnel}
  n111 = ScalarParameter() {defaultValue=0.35, group=HueShift, parameterName=FresnelHueShiftScale}
  n112 = Frac(None=Multiply(A=Add(A=@n107, B=@n110), B=@n111))
  n113 = Multiply(A=@n110, B=@n111)
  n114 = ComponentMask(None=WorldPosition() {worldPositionShaderOffset=WPT_ExcludeAllShaderOffsets}.XYZ) {g=true, r=true}
  n115 = ScalarParameter() {defaultValue=15500, group=HueShift, parameterName=HueShiftDivisor}
  n116 = Add(
   A=TextureSample!(UVs=Divide(A=@n114, B=@n115) {constB=2000}, Tex, Apply View MipBias) {samplerType=SAMPLERTYPE_LinearColor, texture=/Game/Global/Textures/Default/Noise/T_Noise01_Inverse}.R
   B=ScalarParameter() {defaultValue=-0.65, group=HueShift, parameterName=WorldHueShiftBias}
  )
  n117 = ScalarParameter() {defaultValue=2.5, group=HueShift, parameterName=WorldHueShiftScale}
  n118 = MaterialFunctionCall(
   Color_Morning=VectorParameter() {group=BaseColor, parameterName=HueShiftColor_Morning, defaultValue=<rgba 0.6198 0.6198 0.6198 0>}.RGB
   Color_Day=VectorParameter() {group=BaseColor, parameterName=HueShiftColor_Day, defaultValue=<rgba 0.6439 0.75 0.5378 0>}.RGB
   Color_Evening=VectorParameter() {group=BaseColor, parameterName=HueShiftColor_Evening, defaultValue=<rgba 0.8043 0.8385 0.7701 0>}.RGB
   Color_Night=VectorParameter() {group=BaseColor, parameterName=HueShiftColor_Night, defaultValue=<rgba 0.7113 0.7986 0.6239 0>}.RGB
  ) {materialFunction=&Color_Time_of_Day}
  n119 = MaterialFunctionCall(
   Color_Morning=VectorParameter() {group=BaseColor, parameterName=BaseColor_Morning_Mobile, defaultValue=<rgba 0.3763 0.4288 0.3238 0>}.RGB
   Color_Day=VectorParameter() {group=BaseColor, parameterName=BaseColor_Day_Mobile, defaultValue=<rgba 0.3763 0.4288 0.3238 0>}.RGB
   Color_Evening=VectorParameter() {group=BaseColor, parameterName=BaseColor_Evening_Mobile, defaultValue=<rgba 0.3457 0.3663 0.325 0>}.RGB
   Color_Night=VectorParameter() {group=BaseColor, parameterName=BaseColor_Night_Mobile, defaultValue=<rgba 0.387 0.441 0.333 0>}.RGB
  ) {materialFunction=&Color_Time_of_Day}
  n120 = MaterialFunctionCall(
   Hue Shift Percentage=Add(A=StaticSwitchParameter(True=@n112, False=@n113) {group=HueShift, parameterName=AnimateFresnelHueShift}, B=Multiply(A=@n116, B=@n117))
   Texture=VertexInterpolator(VS=ShadingPathSwitch(Default=@n118.Result, Mobile=@n119.Result)).PS
  ) {materialFunction=&HueShift}
  n121 = Add(
   A=Add(A=Multiply(A=@n12, B=@n13), B=@n8.RGB)
   B=Multiply(
    A=FunctionInput(Preview=@n120.Result) {bUsePreviewValueAsDefault=true, inputName=BaseColorOverride, sortPriority=2}
    B=ScalarParameter() {defaultValue=0.5, parameterName=BaseColorOverrideIntensity}
   )
  )
  n122 = Multiply(A=NamedRerouteUsage(), B=NamedRerouteUsage())
  NamedRerouteDeclaration(None=Desaturation(None=Add(A=@n121, B=@n122), Fraction=ScalarParameter() {defaultValue=-0.7, parameterName=BaseColorDesaturationFraction})) {name=BaseColor, nodeColor=<rgba 0 0.6484 1 1>}
  n123 = ScalarParameter() {defaultValue=0.28, parameterName=Roughness.-Z}
  n124 = ScalarParameter() {defaultValue=0.22, parameterName=Roughness}
  n125 = NamedRerouteUsage()
  NamedRerouteDeclaration(None=LinearInterpolate(
    A=LinearInterpolate(A=@n123, B=@n124, Alpha=@n125)
    B=ScalarParameter() {defaultValue=0.9, parameterName=RoughnessSpecRemoval}
    Alpha=NamedRerouteUsage()
   )) {name=Roughness, nodeColor=<rgba 1 0.3984 0 1>}
  n126 = ScalarParameter() {defaultValue=260, parameterName=RippleUVDivisor}
  n127 = Divide(A=NamedRerouteUsage(), B=@n126) {constB=1600}
  n128 = ComponentMask(None=@n127) {r=true}
  n129 = Constant() {r=6}
  n130 = Add(
   A=@n127
   B=AppendVector(
    A=Constant()
    B=MaterialFunctionCall!(PrimaryTextureCoordinates=@n128, 2ndOrderTextureCoordinates, RandomSeed=@n129, NoiseDensityBias, Bypass Noise Influence Crosstalk) {materialFunction=/Game/Effects/Fort_Effects/Materials/_Functions/MF_RandomUVSample}.NoiseWithRange
   )
  )
  n131 = NamedRerouteDeclaration(None=ConstantBiasScale(None=Frac(None=@n130)) {bias=-0.5, scale=2}) {name=Ripple_UVs_A, nodeColor=<rgba 0.8281 0 1 1>}
  n132 = StaticBool() {value=true}
  n133 = ScalarParameter() {defaultValue=2, group=Chrome_Phased, parameterName=Phased_U_Scale}
  n134 = ScalarParameter() {defaultValue=0.6, group=Chrome_Phased, parameterName=Phased_V_Scale}
  n135 = MaterialFunctionCall(Float2=Multiply(
    A=MaterialFunctionCall(Vector or UVs=@n131, Swizzle Coordinate Output=@n132) {materialFunction=&VectorToRadialValue}.Radial Coordinates
    B=AppendVector(A=@n133, B=@n134)
   )) {materialFunction=&BreakOutFloat2Components}
  n136 = ScalarParameter() {defaultValue=3, group=Chrome_Phased, parameterName=Phased_V_Exponent}
  n137 = ConstantBiasScale(None=MaterialFunctionCall(PrimaryTextureCoordinates=@n130, RandomSeed=Constant() {r=4}) {materialFunction=&MF_RandomUVSample}.NoiseWithRange)
  n138 = MaterialFunctionCall(X=@n135.R, Y=Power(Base=@n135.G, Exp=Multiply(A=@n136, B=@n137)) {constExponent=2.5}) {materialFunction=&MakeFloat2}
  n139 = Multiply(A=Time(), B=ScalarParameter() {defaultValue=1, parameterName=PhasedRipplePanSpeedScale})
  n140 = Multiply(A=@n137, B=@n139) {constA=1}
  n141 = TextureSample!(
   UVs=NamedRerouteDeclaration(None=Panner(Coordinate=@n138.Result, Time=@n140) {bFractionalPart=true, speedX=0.0803, speedY=0.078}) {name=PanningRadial.1, nodeColor=<rgba 1 0.7969 0 1>}
   Tex
   "DDX(UVs)"=NamedRerouteUsage()
   "DDY(UVs)"=NamedRerouteUsage()
   Apply View MipBias
  ) {
   mipValueMode=TMVM_Derivative
   samplerSource=SSM_Wrap_WorldGroupSettings
   samplerType=SAMPLERTYPE_Normal
   texture=&T_Lackluster_Phased_Normal
  }
  n142 = TextureSample(
   UVs=Panner(Coordinate=Multiply(A=@n138.Result) {constB=-1}, Time=@n140) {bFractionalPart=true, speedX=0.115, speedY=-0.0244}
   "DDX(UVs)"=NamedRerouteUsage()
   "DDY(UVs)"=NamedRerouteUsage()
  ) {
   mipValueMode=TMVM_Derivative
   samplerSource=SSM_Wrap_WorldGroupSettings
   samplerType=SAMPLERTYPE_Normal
   texture=&T_Lackluster_Phased_Normal
  }
  n143 = Abs(None=NamedRerouteUsage())
  n144 = SphereMask(A=Power(Base=@n143) {constExponent=3}, B=Constant(), Radius=Constant() {r=1}, Hardness=Constant() {r=0.7})
  n145 = OneMinus(None=SphereMask(A=@n143, B=Constant(), Radius=Constant() {r=0.4}, Hardness=Constant() {r=0.4}))
  n146 = LinearInterpolate(
   A=Constant3Vector() {constant=<rgba 0 0 1 0>}
   B=Add(A=@n141.RGB, B=@n142.RGB) {constB=0}
   Alpha=NamedRerouteDeclaration(None=Multiply(A=@n144, B=@n145)) {name=RippleMaskA, nodeColor=<rgba 1 0.7031 0 1>}
  )
  n147 = Divide(A=NamedRerouteUsage(), B=Divide(A=@n126) {constB=0.666}) {constB=1066}
  n148 = ComponentMask(None=@n147) {r=true}
  n149 = Constant() {r=9}
  n150 = Add(
   A=@n147
   B=AppendVector(
    A=Constant()
    B=MaterialFunctionCall(PrimaryTextureCoordinates=@n148, RandomSeed=@n149) {materialFunction=&MF_RandomUVSample}.NoiseWithRange
   )
  )
  n151 = NamedRerouteDeclaration(None=ConstantBiasScale(None=Frac(None=@n150)) {bias=-0.5, scale=2}) {name=Ripple_UVs_B, nodeColor=<rgba 0 0.9062 1 1>}
  n152 = StaticBool() {value=true}
  n153 = ScalarParameter() {defaultValue=2, group=Chrome_Phased, parameterName=Phased_U_Scale}
  n154 = ScalarParameter() {defaultValue=0.6, group=Chrome_Phased, parameterName=Phased_V_Scale}
  n155 = MaterialFunctionCall(Float2=Multiply(
    A=MaterialFunctionCall(Vector or UVs=@n151, Swizzle Coordinate Output=@n152) {materialFunction=&VectorToRadialValue}.Radial Coordinates
    B=AppendVector(A=@n153, B=@n154)
   )) {materialFunction=&BreakOutFloat2Components}
  n156 = ScalarParameter() {defaultValue=3, group=Chrome_Phased, parameterName=Phased_V_Exponent}
  n157 = MaterialFunctionCall(PrimaryTextureCoordinates=@n150, 2ndOrderTextureCoordinates=Constant() {r=6}) {materialFunction=&MF_RandomUVSample}
  n158 = MaterialFunctionCall(X=@n155.R, Y=Power(Base=@n155.G, Exp=Multiply(A=@n156, B=@n157.NoiseWithRange)) {constExponent=2.5}) {materialFunction=&MakeFloat2}
  n159 = Multiply(A=ConstantBiasScale(None=@n157.NoiseWithRange), B=@n139) {constA=1}
  n160 = TextureSample(
   UVs=NamedRerouteDeclaration(None=Panner(Coordinate=@n158.Result, Time=@n159) {bFractionalPart=true, speedX=0.1605, speedY=0.0312}) {name=PanningRadial.2, nodeColor=<rgba 1 0.5391 0 1>}
   "DDX(UVs)"=NamedRerouteUsage()
   "DDY(UVs)"=NamedRerouteUsage()
  ) {
   mipValueMode=TMVM_Derivative
   samplerSource=SSM_Wrap_WorldGroupSettings
   samplerType=SAMPLERTYPE_Normal
   texture=&T_Lackluster_Phased_Normal
  }
  n161 = TextureSample(
   UVs=Panner(Coordinate=Multiply(A=@n158.Result) {constB=-1}, Time=@n159) {bFractionalPart=true, speedX=0.115, speedY=-0.0205}
   "DDX(UVs)"=NamedRerouteUsage()
   "DDY(UVs)"=NamedRerouteUsage()
  ) {
   mipValueMode=TMVM_Derivative
   samplerSource=SSM_Wrap_WorldGroupSettings
   samplerType=SAMPLERTYPE_Normal
   texture=&T_Lackluster_Phased_Normal
  }
  n162 = Abs(None=NamedRerouteUsage())
  n163 = SphereMask(A=Power(Base=@n162) {constExponent=1.5}, B=Constant(), Radius=Constant() {r=1}, Hardness=Constant() {r=0.7})
  n164 = OneMinus(None=SphereMask(A=@n162, B=Constant(), Radius=Constant() {r=0.4}, Hardness=Constant() {r=0.4}))
  n165 = LinearInterpolate(
   A=Constant3Vector() {constant=<rgba 0 0 1 0>}
   B=Add(A=@n160.RGB, B=@n161.RGB)
   Alpha=NamedRerouteDeclaration(None=Multiply(A=@n163, B=@n164)) {name=RippleMaskB, nodeColor=<rgba 0 1 0.7891 1>}
  )
  NamedRerouteDeclaration(None=ShadingPathSwitch(Default=Add(A=@n146, B=@n165), Mobile=@n146)) {name=PhasedNormals, nodeColor=<rgba 0 0.8594 1 1>}
  n166 = ScalarParameter() {defaultValue=0.97, parameterName=Metallic}
  n167 = ScalarParameter() {defaultValue=0.85, parameterName=Metallic_Mobile}
  NamedRerouteDeclaration(None=LinearInterpolate(
    A=ShadingPathSwitch(Default=@n166, Mobile=@n167)
    B=ScalarParameter() {defaultValue=0.1, parameterName=MetallicSpecRemoval}
    Alpha=NamedRerouteUsage()
   )) {name=Metallic, nodeColor=<rgba 0.5 1 0 1>}
  n168 = Divide(A=NamedRerouteUsage()) {constB=1}
  n169 = Multiply(A=ScalarParameter() {defaultValue=1, parameterName=ChromePanSpeedScale}, B=Time())
  NamedRerouteDeclaration(None=Panner(Coordinate=NamedRerouteDeclaration(None=@n168) {name=Chrome_UVs_A_NoPanning, nodeColor=<rgba 0 1 0.7656 1>}, Time=@n169) {bFractionalPart=true, speedX=0.0008, speedY=0.0022}) {name=Chrome_UVs_A, nodeColor=<rgba 0.9219 0 1 1>}
  n170 = NamedRerouteUsage()
  NamedRerouteDeclaration(None=Panner(Coordinate=Divide(A=@n170) {constB=1}, Time=@n169) {bFractionalPart=true, speedX=0.0054, speedY=0.0022}) {name=Chrome_UVs_B, nodeColor=<rgba 0 0.3906 1 1>}
  NamedRerouteDeclaration(None=Transform(None=@n49)) {name=ChromeNormalsWS, nodeColor=<rgba 0.8516 0 1 1>}
  n171 = TextureSampleParameter2D() {
   parameterName=Normals
   samplerSource=SSM_Wrap_WorldGroupSettings
   samplerType=SAMPLERTYPE_Normal
   texture=&FlatNormalBlue
  }
  n172 = TextureSampleParameter2D() {
   parameterName=Normals_Texture_2
   samplerSource=SSM_Wrap_WorldGroupSettings
   samplerType=SAMPLERTYPE_Normal
   sortPriority=-5
   texture=&FlatNormalBlue
  }
  n173 = TextureSampleParameter2D() {
   parameterName=Normals_Texture_3
   samplerSource=SSM_Wrap_WorldGroupSettings
   samplerType=SAMPLERTYPE_Normal
   sortPriority=-5
   texture=&FlatNormalBlue
  }
  n174 = TextureSampleParameter2D() {
   parameterName=Normals_Texture_4
   samplerSource=SSM_Wrap_WorldGroupSettings
   samplerType=SAMPLERTYPE_Normal
   sortPriority=-5
   texture=/Engine/EngineMaterials/FlatNormal
  }
  n175 = StaticBool() {value=true}
  n176 = StaticSwitchParameter(
   True=MaterialFunctionCall(1=@n171.RGB, 2=@n172.RGB, 3=@n173.RGB, 4=@n174.RGB, Use 2 Textures=@n175, Use 3 Textures=@n175, Use 4 Textures=@n175) {materialFunction=&MF_LerpTexturesBasedOnInt_Float3_4Layers}.Result
   False=@n171.RGB
  ) {parameterName=Use4Layers}
  n177 = LinearInterpolate(
   A=GetMaterialAttributes(None=NamedRerouteUsage()) {attributeGetTypes [#6]}.Normal
   B=@n176
   Alpha=NamedRerouteUsage()
  )
  n178 = NamedRerouteUsage()
  n179 = VectorParameter() {parameterName=SourceNormalsStrength, defaultValue=<rgba 0.5 0.5 1 0>}
  NamedRerouteDeclaration(None=Multiply(
    A=StaticSwitch(True=@n177, False=@n176, Value=@n178)
    B=FunctionInput(Preview=@n179.RGB) {bUsePreviewValueAsDefault=true, inputName=SourceNormalsStrengthOverride, sortPriority=5}
   )) {name=NormalsScaled, nodeColor=<rgba 0.5469 1 0 1>}
  NamedRerouteDeclaration(None=FunctionInput(Preview=VertexNormalWS()) {bUsePreviewValueAsDefault=true, inputName=VertexNormalOverride, sortPriority=4}) {name=VertexNormal, nodeColor=<rgba 0.5 0 1 1>}
  ScalarParameter() {defaultValue=1290.24, parameterName=Chrome_Divisor_UV_a}
  ScalarParameter() {defaultValue=2257.9199, parameterName=Chrome_Divisor_UV_b}
  Subtract(A=WorldPosition().XYZ, B=ObjectPositionWS())
  VertexInterpolator(VS=PreSkinnedNormal())
 ]}
