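# Character material for the LackLuster seasonal blob look, built for skeletal and deformer-driven meshes with a chrome physical material.
# All shading comes from a single MF_LackLuster_Blob function call using material attributes.
# Pre-skinned position and normal pass through vertex interpolators into its world-position and vertex-normal overrides, so the function works from rest-pose mesh data rather than animated world space.
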Material /Game/TechArt/Seasonal/LackLuster/Materials/M_LackLuster_Character_Blob
graph {
 properties {
  bOutputTranslucentVelocity=true
  bUseMaterialAttributes=true
  bUsedWithMeshDeformer=true
  bUsedWithSkeletalMesh=true
  physMaterial=/Game/TechArt/Seasonal/LackLuster/Physics/Chrome_PhysicalMaterial
  pixelDepthOffsetMode=PDOM_Legacy
  usedShadingModels=Default Lit
 }
 nodes [MaterialFunctionCall!(
   bBlendMaterialAttributes
   MA_In
   UVsOverride
   Override Normals
   BlendMaterialAttributesAlpha
   BaseColorOverride
   WorldPositionOverride=VertexInterpolator(VS=PreSkinnedPosition()).PS
   VertexNormalOverride=VertexInterpolator(VS=PreSkinnedNormal()).PS
   SourceNormalsStrengthOverride
  ) {materialFunction=/Game/TechArt/Seasonal/LackLuster/MaterialFunctions/MF_LackLuster_Blob}]
}
