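# Surface material for the Jungle variant of the seasonal LackLuster effect, usable on instanced static and spline meshes, with a chrome physical material.
# MF_LackLuster_Base does the main work, with its attribute-blend static switch defaulting on and vertex color red as the blend alpha.
# Its input attributes are an empty SetMaterialAttributes node, routed through MF_VPTexture only on the deferred shading path via a ShadingPathSwitch.
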Material /Game/TechArt/Seasonal/LackLuster/Materials/M_LackLuster_Jungle
graph {
 properties {
  bOutputTranslucentVelocity=true
  bUseMaterialAttributes=true
  bUsedWithInstancedStaticMeshes=true
  bUsedWithSplineMeshes=true
  ditherOpacityMask=true
  opacityMaskClipValue=0.5
  physMaterial=/Game/TechArt/Seasonal/LackLuster/Physics/Chrome_PhysicalMaterial
  pixelDepthOffsetMode=PDOM_Legacy
  usedShadingModels=Default Lit
 }
 nodes [
  n1 = SetMaterialAttributes!(MaterialAttributes)
  MaterialFunctionCall!(
   bBlendMaterialAttributes=StaticBoolParameter() {defaultValue=true, parameterName=BlendMaterialAttributes}
   MA_In=ShadingPathSwitch(Default=@n1, Deferred=MaterialFunctionCall!(In=@n1, UseOneSampler) {materialFunction=/Game/Athena/WinterArt/Materials/MF_VPTexture}.Result)
   UVsOverride
   Override Normals
   BlendMaterialAttributesAlpha=VertexColor().R
   BaseColorOverride
   WorldPositionOverride
   VertexNormalOverride
   SourceNormalsStrengthOverride
  ) {materialFunction=/Game/TechArt/Seasonal/LackLuster/MaterialFunctions/MF_LackLuster_Base}
 ]
}
