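# Parent material for the seasonal LackLuster look, in a NoDither variant for static, instanced, skeletal and spline meshes, tagged with a Chrome physical material.
# MF_LackLuster_Base builds the material attributes, with none of its override inputs wired.
# Its output passes through the shared MF_ChestAnimation function, so the graph itself is only these two function calls.
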Material /Game/TechArt/Seasonal/LackLuster/Materials/M_LackLuster_Parent_NoDither
graph {
 properties {
  bOutputTranslucentVelocity=true
  bUseMaterialAttributes=true
  bUsedWithInstancedStaticMeshes=true
  bUsedWithMeshDeformer=true
  bUsedWithSkeletalMesh=true
  bUsedWithSplineMeshes=true
  opacityMaskClipValue=0.5
  physMaterial=/Game/TechArt/Seasonal/LackLuster/Physics/Chrome_PhysicalMaterial
  pixelDepthOffsetMode=PDOM_Legacy
  usedShadingModels=Default Lit
 }
 nodes [MaterialFunctionCall!(In=MaterialFunctionCall!(
    bBlendMaterialAttributes
    MA_In
    UVsOverride
    Override Normals
    BlendMaterialAttributesAlpha
    BaseColorOverride
    WorldPositionOverride
    VertexNormalOverride
    SourceNormalsStrengthOverride
   ) {materialFunction=/Game/TechArt/Seasonal/LackLuster/MaterialFunctions/MF_LackLuster_Base}.MA Out) {materialFunction=/Game/Packages/Fortress_SharedMaterials/Base_Material_Functions/MF_ChestAnimation}]
}
