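# Material function that adds music-reactive emissive to LackLuster petrified material attributes, gated by the UsePetrifiedFFT static switch.
# Audio spectrum collection parameters feed reactivity values that scale base emissive and a colored trim glow masked by world-normal math.
# A Frequency Selector, randomizable through custom primitive data, feeds band selection, with a static switch for debug frequency colors.
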MaterialFunction /Game/TechArt/Seasonal/LackLuster/MaterialFunctions/MF_LackLuster_Petrified_FFT
graph {nodes [
  n1 = FunctionInput() {inputType=FunctionInput_MaterialAttributes}
  n2 = GetMaterialAttributes(None=@n1) {attributeGetTypes [#1]}
  n3 = LinearInterpolate(
   A=ScalarParameter() {defaultValue=0.1, parameterName=Base Emissive Min}
   B=ScalarParameter() {defaultValue=2, parameterName=Base Emissive Max}
   Alpha=NamedRerouteUsage()
  ) {constA=1, constAlpha=0}
  n4 = Multiply(A=NamedRerouteUsage(), B=NamedRerouteUsage())
  n5 = Multiply(
   A=ScalarParameter() {defaultValue=1.1, parameterName=Trim Reactivity Bias}
   B=LinearInterpolate(A=NamedRerouteUsage(), B=NamedRerouteUsage()) {constAlpha=0.2}
  )
  n6 = LinearInterpolate(
   A=ScalarParameter() {defaultValue=0.5, parameterName=Trim Min}
   B=ScalarParameter() {defaultValue=5, parameterName=Trim Max}
   Alpha=Power(Base=Saturate(None=@n5))
  ) {constA=1}
  n7 = StaticSwitchParameter(True=NamedRerouteUsage(), False=Add(A=Multiply(A=@n2.EmissiveColor, B=@n3), B=Multiply(A=@n4, B=@n6))) {parameterName=Debug Frequency}
  FunctionOutput(None=StaticSwitchParameter(
    True=SetMaterialAttributes!(MaterialAttributes=@n2.MaterialAttributes, EmissiveColor=@n7) {attributeSetTypes [#1]}
    False=@n1
   ) {parameterName=UsePetrifiedFFT})
  n8 = ScalarParameter() {desc="0 is random, 1 is high, 4 is low", parameterName=Frequency Selector}
  n9 = Subtract(A=@n8)
  n10 = ScalarParameter() {bUseCustomPrimitiveData=true, parameterName=Random}
  n11 = Divide(A=Floor(None=Multiply(A=@n10) {constB=4})) {constB=3}
  NamedRerouteDeclaration(None=If(A=@n8, "A > B"=Divide(A=@n9) {constB=3}, "A == B"=@n11, "A < B"=@n11)) {name=Frequency Selector, nodeColor=<rgba 0.875 1 0 1>}
  NamedRerouteDeclaration(None=VectorParameter() {parameterName=Color 1, defaultValue=<rgba 0.4806 0.1407 1 0>}.RGB) {name=Trim Glow Color, nodeColor=<rgba 1 0 0.7734 1>}
  n12 = VertexInterpolator(VS=VertexNormalWS())
  n13 = Constant3Vector() {constant=<rgba 1 1 0 0>}
  n14 = Multiply(A=DotProduct(A=Constant3Vector() {constant=<rgba 0 0 1 0>}, B=@n12.PS), B=DotProduct(A=@n12.PS, B=Multiply(A=@n12.PS, B=@n13)))
  NamedRerouteDeclaration(None=Saturate(None=InverseLinearInterpolate(Value=@n14) {constA=0.1, constB=0.5})) {name=Trim Mask, nodeColor=<rgba 0 0.1797 1 1>}
  n15 = Constant4Vector()
  n16 = CollectionParameter() {collection=/Game/Athena/Events/AudioAnalysis/MPC_Audio_Reactive_Parameters, parameterName=Spectrum_MusicManualSet}
  n17 = CollectionParameter() {collection=&MPC_Audio_Reactive_Parameters, parameterName=Spectrum_Music}
  n18 = Saturate(None=LinearInterpolate(A=@n16, B=Multiply(A=@n16, B=@n17), Alpha=NamedRerouteUsage()))
  NamedRerouteDeclaration(None=DotProduct(A=@n18, B=Normalize(VectorInput=@n15.RGBA))) {name=Trim Extra Reactivity, nodeColor=<rgba 0 0.2031 1 1>}
  NamedRerouteDeclaration(None=MaterialFunctionCall(
    1=Constant3Vector() {constant=<rgba 1 0 0 0>}
    2=Constant3Vector() {constant=<rgba 1 0 0.6881 1>}
    3=Constant3Vector() {constant=<rgba 0 1 0.2118 0>}
    4=Constant3Vector() {constant=<rgba 0.0107 0 1 0>}
    A=NamedRerouteUsage()
   ) {materialFunction=&Lerp_Multiple_Float3}.Lerp 4 Inputs) {name=Debug Frequency, nodeColor=<rgba 1 0 0.9141 1>}
  n19 = MaterialFunctionCall(Float4=@n18) {materialFunction=&BreakOutFloat4Components}
  n20 = NamedRerouteUsage()
  NamedRerouteDeclaration(None=Saturate(None=MaterialFunctionCall!(1=@n19.R, 2=@n19.R, 3=@n19.R, 4=@n19.R, A=@n20) {materialFunction=/Engine/Functions/Engine_MaterialFunctions02/Lerp_Multiple_Float}.Lerp 4 Inputs)) {name=Audio Reactivity, nodeColor=<rgba 0 1 0.3672 1>}
 ]}
