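# Parent material for the petrified surface look of the LackLuster seasonal content, with a Chrome petrified physical material assigned.
# All shading comes from material functions: MF_LackLuster_Petrified builds material attributes, which feed MF_LackLuster_Petrified_FFT for the final output.
# It uses Default Lit shading with a dithered opacity mask, outputs translucent velocity, and supports Nanite, instanced static meshes and spline meshes.
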Material /Game/TechArt/Seasonal/LackLuster/Materials/M_LackLuster_Petrified_Parent
graph {
 properties {
  bOutputTranslucentVelocity=true
  bUseMaterialAttributes=true
  bUsedWithInstancedStaticMeshes=true
  bUsedWithNanite=true
  bUsedWithSplineMeshes=true
  ditherOpacityMask=true
  physMaterial=/Game/TechArt/Seasonal/LackLuster/Physics/Chrome_Petrified_PhysicalMaterial
  pixelDepthOffsetMode=PDOM_Legacy
  usedShadingModels=Default Lit
 }
 nodes [MaterialFunctionCall!(In=MaterialFunctionCall!(bBlendMaterialAttributes, MA_In, BlendMaterialAttributesAlpha, BaseColorOverride, VertexNormalOverride, Texture Override Position) {materialFunction=/Game/TechArt/Seasonal/LackLuster/MaterialFunctions/MF_LackLuster_Petrified}.MA Out) {materialFunction=/Game/TechArt/Seasonal/LackLuster/MaterialFunctions/MF_LackLuster_Petrified_FFT}]
}
