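# Translucent deferred decal material for the LackLuster seasonal Chrome phasing effect, drawing a glowing, swirling patch.
# Rotating water normals join phased normals and distort UVs, which become radial coordinates for counter-panning color and normal samples, with explicit derivatives.
# Sphere masks shape the opacity falloff and boosted emissive core, alongside timestamp in/out and decal lifetime nodes.
# Two noise-distorted, randomly offset ripple layers build a masked radial normal.
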Material /Game/TechArt/Seasonal/LackLuster/Materials/M_Lackluster_Phasing
graph {
 properties {blendMode=BLEND_Translucent, materialDomain=MD_DeferredDecal, pixelDepthOffsetMode=PDOM_Legacy, usedShadingModels=Default Lit}
 nodes [
  n1 = LightVector()
  n2 = Multiply(
   A=Multiply(A=NamedRerouteUsage(), B=ScalarParameter() {defaultValue=0.075, parameterName=WaterRippleDistortionScale})
   B=Constant3Vector() {constant=<rgba 1 1 0.15 0>}
  )
  n3 = SphereMask(A=@n1, B=Constant() {r=0.5}, Radius=Constant() {r=0.5}, Hardness=Constant() {r=0.5})
  n4 = NamedRerouteDeclaration(None=Add(A=@n1, B=Multiply(A=@n2, B=@n3))) {name=DistortedUVs, nodeColor=<rgba 0 0.6016 1 1>}
  ComponentMask(None=@n4) {g=true, r=true}
  ConstantBiasScale(None=Sine(None=Time()))
  Constant()
  Constant() {r=1}
  MaterialFunctionCall(
   TimestampA=ScalarParameter() {parameterName=TimestampA}
   TransitionLength_In=ScalarParameter() {defaultValue=0.32, parameterName=TransitionLength_In}
  ) {materialFunction=&MF_Timestamp_InOut_01}
  MaterialFunctionCall() {materialFunction=/Game/Weapons/Prototype/SprayGun/Materials/MF_DecalFacingMask}
  Multiply()
  n5 = TextureCoordinate()
  NamedRerouteDeclaration(None=DDX(Value=@n5)) {name=TexCoord_DDX, nodeColor=<rgba 0.7109 1 0 1>}
  NamedRerouteDeclaration(None=DDY(Value=@n5)) {name=TexCoord_DDY, nodeColor=<rgba 1 0 0.3281 1>}
  NamedRerouteDeclaration(None=MaterialFunctionCall!(
    UseLocalSpace
    WorldPosition
    NormalOverride=MaterialFunctionCall() {materialFunction=/Game/Weapons/Prototype/SprayGun/Materials/MF_DecalNormal}.WorldNormal
   ) {materialFunction=/Game/Environments/Apollo/Sets/Skyfire/Materials/MaterialFunctions/MF_WorldAlignedUVs_HardEdge}.Result) {name=WorldUVs, nodeColor=<rgba 0 0.0156 1 1>}
  n6 = NamedRerouteUsage()
  n7 = SphereMask(
   A=NamedRerouteUsage()
   B=Constant() {r=0.5}
   Radius=LinearInterpolate(B=ScalarParameter() {defaultValue=0.5, parameterName=FinalMaskRadius}, Alpha=SmoothStep(Value=@n6) {constMax=0.5}) {constA=0.2}
   Hardness=ScalarParameter() {defaultValue=0.6, parameterName=FinalMaskHardness}
  )
  n8 = ScalarParameter() {defaultValue=5, group=Chrome_Phased, parameterName=FinalMaskExponent}
  NamedRerouteDeclaration(None=Multiply(A=Power(Base=@n7, Exp=@n8) {constExponent=3}, B=NamedRerouteUsage())) {name=Opacity, nodeColor=<rgba 0.4531 0 1 1>}
  n9 = ConstantBiasScale(None=@n4) {bias=-0.5, scale=2}
  n10 = StaticBool() {value=true}
  n11 = ScalarParameter() {defaultValue=1, group=Chrome_Phased, parameterName=Center_U_Scale}
  n12 = ScalarParameter() {defaultValue=0.75, group=Chrome_Phased, parameterName=Center_V_Scale}
  n13 = MaterialFunctionCall(Float2=Multiply(
    A=MaterialFunctionCall(Vector or UVs=@n9, Swizzle Coordinate Output=@n10) {materialFunction=&VectorToRadialValue}.Radial Coordinates
    B=AppendVector(A=@n11, B=@n12)
   )) {materialFunction=&BreakOutFloat2Components}
  n14 = MaterialFunctionCall(
   X=@n13.R
   Y=Power(Base=@n13.G, Exp=ScalarParameter() {defaultValue=2, group=Chrome_Phased, parameterName=Center_V_Exponent}) {constExponent=2.5}
  ) {materialFunction=&MakeFloat2}
  n15 = Panner(Coordinate=@n14.Result) {bFractionalPart=true, speedX=0.1605, speedY=0.2}
  n16 = TextureSample!(UVs=@n15, Tex, "DDX(UVs)"=NamedRerouteUsage(), "DDY(UVs)"=NamedRerouteUsage(), Apply View MipBias) {
   mipValueMode=TMVM_Derivative
   samplerSource=SSM_Wrap_WorldGroupSettings
   samplerType=SAMPLERTYPE_Normal
   texture=/Game/TechArt/Seasonal/LackLuster/Textures/T_Lackluster_Phased_Normal
  }
  n17 = Panner(Coordinate=Multiply(A=@n14.Result) {constB=-1}) {bFractionalPart=true, speedX=0.115, speedY=-0.13}
  n18 = TextureSample(UVs=@n17, "DDX(UVs)"=NamedRerouteUsage(), "DDY(UVs)"=NamedRerouteUsage()) {
   mipValueMode=TMVM_Derivative
   samplerSource=SSM_Wrap_WorldGroupSettings
   samplerType=SAMPLERTYPE_Normal
   texture=&T_Lackluster_Phased_Normal
  }
  n19 = TextureCoordinate() {uTiling=0.3, vTiling=0.3}
  n20 = TextureSampleParameter2D(UVs=Rotator(Coordinate=@n19) {centerX=0.2225, centerY=0.2225, speed=-0.2081}) {parameterName=Decal Normal, samplerType=SAMPLERTYPE_Normal, texture=/Game/Global/Textures/Default/Tileable_Patterns/T_WaterSoft_N}
  n21 = TextureSampleParameter2D(UVs=Rotator(Coordinate=@n19) {centerX=0.2225, centerY=0.2225, speed=0.1665}) {parameterName=Decal Normal, samplerType=SAMPLERTYPE_Normal, texture=&T_WaterSoft_N}
  n22 = Add(A=Add(A=@n20.RGB, B=@n21.RGB)) {constB=0}
  n23 = Add(
   A=Add(A=@n16.RGB, B=@n18.RGB)
   B=Multiply(A=NamedRerouteDeclaration(None=@n22) {name=WaterRippleNormals, nodeColor=<rgba 0 1 0.2031 1>}, B=Constant3Vector() {constant=<rgba 1 1 0 0>})
  )
  n24 = Multiply(A=Constant3Vector() {constant=<rgba 1 1 -1 0>}, B=@n23)
  n25 = MaterialFunctionCall() {materialFunction=&MF_DecalFacingMask}
  NamedRerouteDeclaration(None=Normalize(VectorInput=LinearInterpolate(A=@n24, B=@n23, Alpha=@n25.Result))) {name=Normal, nodeColor=<rgba 1 0.4688 0 1>}
  n26 = TextureSampleParameter2D!(UVs=@n15, "DDX(UVs)"=NamedRerouteUsage(), "DDY(UVs)"=NamedRerouteUsage(), Apply View MipBias) {
   group=Chrome_Phased
   mipValueMode=TMVM_Derivative
   parameterName=PhasedMap
   texture=/Game/TechArt/Seasonal/LackLuster/Textures/T_Lackluster_Phased
  }
  n27 = TextureSampleParameter2D(UVs=@n17, "DDX(UVs)"=NamedRerouteUsage(), "DDY(UVs)"=NamedRerouteUsage()) {
   group=Chrome_Phased
   mipValueMode=TMVM_Derivative
   parameterName=PhasedMap
   texture=&T_Lackluster_Phased
  }
  n28 = Add(
   A=VectorParameter() {group=Chrome_Phased, parameterName=PhasedColorBias, defaultValue=<rgba 0.06 0.0981 0.2031 0>}.RGB
   B=Desaturation(None=Add(A=@n26.RGB, B=@n27.RGB), Fraction=OneMinus(None=@n3))
  )
  n29 = Multiply(A=@n28, B=ScalarParameter() {defaultValue=15, group=Chrome_Phased, parameterName=PhasedEmissiveScale})
  n30 = NamedRerouteUsage()
  n31 = NamedRerouteUsage()
  n32 = Constant() {r=0.5}
  n33 = ScalarParameter() {defaultValue=0.2, parameterName=CoreMaskRadius}
  n34 = ScalarParameter() {defaultValue=0.2, parameterName=CoreMaskHarhness}
  n35 = Multiply(
   A=ConstantBiasScale(None=@n30) {bias=-0.6, scale=2.5}
   B=Power(Base=SphereMask(A=@n31, B=@n32, Radius=@n33, Hardness=@n34), Exp=ScalarParameter() {defaultValue=3, parameterName=CoreMaskExponent})
  )
  n36 = ConstantBiasScale(None=@n30) {bias=-0.3, scale=1.4286}
  NamedRerouteDeclaration(None=Multiply(A=LinearInterpolate(A=@n28, B=@n29, Alpha=@n35), B=Saturate(None=@n36))) {name=Emissive Color, nodeColor=<rgba 0.0312 0 1 1>}
  NamedRerouteDeclaration(None=DecalLifetimeOpacity().Opacity) {name=Decal Lifetime Opacity, nodeColor=<rgba 0.875 0 1 1>}
  NamedRerouteUsage()
  NamedRerouteUsage()
  NamedRerouteUsage()
  NamedRerouteUsage()
  n37 = NamedRerouteUsage()
  n38 = ScalarParameter() {defaultValue=1250, parameterName=DistortionUVDivisor_A}
  n39 = ScalarParameter() {defaultValue=0.0315, parameterName=DistortionUV_PanSpeedU_A}
  n40 = ScalarParameter() {defaultValue=0.0192, parameterName=DistortionUV_PanSpeedV_A}
  n41 = TextureSample!(
   UVs=Panner(Coordinate=Divide(A=@n37, B=@n38), Speed=AppendVector(A=@n39, B=@n40)) {bFractionalPart=true, speedX=0.06, speedY=0.05}
   Tex
   Apply View MipBias
  ) {
   samplerSource=SSM_Wrap_WorldGroupSettings
   samplerType=SAMPLERTYPE_Normal
   texture=/Game/Gadgets/Assets/Telephone_Booth/Textures/T_BlindsNormalNoise_N
  }
  n42 = ScalarParameter() {defaultValue=260, parameterName=RippleUVDivisor}
  n43 = Add(
   A=Multiply(A=ScalarParameter() {defaultValue=0.1, parameterName=DistortionUVScale_A}, B=ComponentMask(None=@n41.RGB) {g=true, r=true})
   B=Divide(A=NamedRerouteUsage(), B=@n42) {constB=1600}
  )
  n44 = ComponentMask(None=@n43) {r=true}
  n45 = Constant() {r=6}
  n46 = Add(
   A=@n43
   B=AppendVector(
    A=Constant()
    B=MaterialFunctionCall!(PrimaryTextureCoordinates=@n44, 2ndOrderTextureCoordinates, RandomSeed=@n45, NoiseDensityBias, Bypass Noise Influence Crosstalk) {materialFunction=/Game/Effects/Fort_Effects/Materials/_Functions/MF_RandomUVSample}.NoiseWithRange
   )
  )
  n47 = NamedRerouteDeclaration(None=ConstantBiasScale(None=Frac(None=@n46)) {bias=-0.5, scale=2}) {name=Ripple_UVs_A, nodeColor=<rgba 0.8281 0 1 1>}
  n48 = StaticBool() {value=true}
  n49 = ScalarParameter() {defaultValue=1, group=Chrome_Phased, parameterName=Phased_U_Scale}
  n50 = ScalarParameter() {defaultValue=0.5, group=Chrome_Phased, parameterName=Phased_V_Scale}
  n51 = MaterialFunctionCall(Float2=Multiply(
    A=MaterialFunctionCall(Vector or UVs=@n47, Swizzle Coordinate Output=@n48) {materialFunction=&VectorToRadialValue}.Radial Coordinates
    B=AppendVector(A=@n49, B=@n50)
   )) {materialFunction=&BreakOutFloat2Components}
  n52 = ScalarParameter() {defaultValue=3, group=Chrome_Phased, parameterName=Phased_V_Exponent}
  n53 = ConstantBiasScale(None=MaterialFunctionCall(PrimaryTextureCoordinates=@n46, RandomSeed=Constant() {r=4}) {materialFunction=&MF_RandomUVSample}.NoiseWithRange)
  n54 = MaterialFunctionCall(X=@n51.R, Y=Power(Base=@n51.G, Exp=Multiply(A=@n52, B=@n53)) {constExponent=2.5}) {materialFunction=&MakeFloat2}
  n55 = Multiply(A=Time(), B=ScalarParameter() {defaultValue=1, parameterName=PhasedRipplePanSpeedScale})
  n56 = Multiply(A=@n53, B=@n55) {constA=1}
  n57 = TextureSample!(
   UVs=NamedRerouteDeclaration(None=Panner(Coordinate=@n54.Result, Time=@n56) {bFractionalPart=true, speedX=0.0803, speedY=0.078}) {name=PanningRadial.1, nodeColor=<rgba 1 0.7969 0 1>}
   Tex
   "DDX(UVs)"=NamedRerouteUsage()
   "DDY(UVs)"=NamedRerouteUsage()
   Apply View MipBias
  ) {
   mipValueMode=TMVM_Derivative
   samplerSource=SSM_Wrap_WorldGroupSettings
   samplerType=SAMPLERTYPE_Normal
   texture=&T_Lackluster_Phased_Normal
  }
  n58 = TextureSample(
   UVs=Panner(Coordinate=Multiply(A=@n54.Result) {constB=-1}, Time=@n56) {bFractionalPart=true, speedX=0.115, speedY=-0.0244}
   "DDX(UVs)"=NamedRerouteUsage()
   "DDY(UVs)"=NamedRerouteUsage()
  ) {
   mipValueMode=TMVM_Derivative
   samplerSource=SSM_Wrap_WorldGroupSettings
   samplerType=SAMPLERTYPE_Normal
   texture=&T_Lackluster_Phased_Normal
  }
  n59 = Abs(None=NamedRerouteUsage())
  n60 = Constant()
  n61 = Constant() {r=0.4}
  n62 = Constant() {r=0.4}
  n63 = Multiply(
   A=SphereMask(A=Power(Base=@n59) {constExponent=3}, B=Constant(), Radius=Constant() {r=1}, Hardness=Constant() {r=0.7})
   B=OneMinus(None=SphereMask(A=@n59, B=@n60, Radius=@n61, Hardness=@n62))
  )
  n64 = NamedRerouteUsage()
  n65 = ScalarParameter() {defaultValue=800, parameterName=DistortionUVDivisor_B}
  n66 = ScalarParameter() {defaultValue=-0.052, parameterName=DistortionUV_PanSpeedU_B}
  n67 = ScalarParameter() {defaultValue=-0.0179, parameterName=DistortionUV_PanSpeedV_B}
  n68 = TextureSample!(
   UVs=Panner(Coordinate=Divide(A=@n64, B=@n65), Speed=AppendVector(A=@n66, B=@n67)) {bFractionalPart=true, speedX=0.06, speedY=0.05}
   Tex
   Apply View MipBias
  ) {
   samplerSource=SSM_Wrap_WorldGroupSettings
   samplerType=SAMPLERTYPE_Normal
   texture=&T_BlindsNormalNoise_N
  }
  n69 = Add(
   A=Divide(A=NamedRerouteUsage(), B=Divide(A=@n42) {constB=0.666}) {constB=1066}
   B=Multiply(A=ScalarParameter() {defaultValue=0.12, parameterName=DistortionUVScale_B}, B=ComponentMask(None=@n68.RGB) {g=true, r=true})
  )
  n70 = ComponentMask(None=@n69) {r=true}
  n71 = Constant() {r=9}
  n72 = Add(
   A=@n69
   B=AppendVector(
    A=Constant()
    B=MaterialFunctionCall(PrimaryTextureCoordinates=@n70, RandomSeed=@n71) {materialFunction=&MF_RandomUVSample}.NoiseWithRange
   )
  )
  n73 = NamedRerouteDeclaration(None=ConstantBiasScale(None=Frac(None=@n72)) {bias=-0.5, scale=2}) {name=Ripple_UVs_B, nodeColor=<rgba 0 0.3672 1 1>}
  n74 = StaticBool() {value=true}
  n75 = ScalarParameter() {defaultValue=1, group=Chrome_Phased, parameterName=Phased_U_Scale}
  n76 = ScalarParameter() {defaultValue=0.5, group=Chrome_Phased, parameterName=Phased_V_Scale}
  n77 = MaterialFunctionCall(Float2=Multiply(
    A=MaterialFunctionCall(Vector or UVs=@n73, Swizzle Coordinate Output=@n74) {materialFunction=&VectorToRadialValue}.Radial Coordinates
    B=AppendVector(A=@n75, B=@n76)
   )) {materialFunction=&BreakOutFloat2Components}
  n78 = ScalarParameter() {defaultValue=3, group=Chrome_Phased, parameterName=Phased_V_Exponent}
  n79 = MaterialFunctionCall(PrimaryTextureCoordinates=@n72, 2ndOrderTextureCoordinates=Constant() {r=6}) {materialFunction=&MF_RandomUVSample}
  n80 = MaterialFunctionCall(X=@n77.R, Y=Power(Base=@n77.G, Exp=Multiply(A=@n78, B=@n79.NoiseWithRange)) {constExponent=2.5}) {materialFunction=&MakeFloat2}
  n81 = Multiply(A=ConstantBiasScale(None=@n79.NoiseWithRange), B=@n55) {constA=1}
  n82 = TextureSample!(
   UVs=NamedRerouteDeclaration(None=Panner(Coordinate=@n80.Result, Time=@n81) {bFractionalPart=true, speedX=0.1605, speedY=0.0312}) {name=PanningRadial.2, nodeColor=<rgba 1 0.5391 0 1>}
   Tex
   "DDX(UVs)"=NamedRerouteUsage()
   "DDY(UVs)"=NamedRerouteUsage()
   Apply View MipBias
  ) {
   mipValueMode=TMVM_Derivative
   samplerSource=SSM_Wrap_WorldGroupSettings
   samplerType=SAMPLERTYPE_Normal
   texture=&T_Lackluster_Phased_Normal
  }
  n83 = TextureSample(
   UVs=Panner(Coordinate=Multiply(A=@n80.Result) {constB=-1}, Time=@n81) {bFractionalPart=true, speedX=0.115, speedY=-0.0205}
   "DDX(UVs)"=NamedRerouteUsage()
   "DDY(UVs)"=NamedRerouteUsage()
  ) {
   mipValueMode=TMVM_Derivative
   samplerSource=SSM_Wrap_WorldGroupSettings
   samplerType=SAMPLERTYPE_Normal
   texture=&T_Lackluster_Phased_Normal
  }
  n84 = Abs(None=NamedRerouteUsage())
  n85 = Constant()
  n86 = Constant() {r=0.4}
  n87 = Constant() {r=0.4}
  n88 = Multiply(
   A=SphereMask(A=Power(Base=@n84) {constExponent=1.5}, B=Constant(), Radius=Constant() {r=1}, Hardness=Constant() {r=0.7})
   B=OneMinus(None=SphereMask(A=@n84, B=@n85, Radius=@n86, Hardness=@n87))
  )
  n89 = Add(
   A=LinearInterpolate(
    A=Constant3Vector() {constant=<rgba 0 0 1 0>}
    B=Add(A=@n57.RGB, B=@n58.RGB) {constB=0}
    Alpha=NamedRerouteDeclaration(None=@n63) {name=RippleMaskA, nodeColor=<rgba 1 0.7031 0 1>}
   )
   B=LinearInterpolate(
    A=Constant3Vector() {constant=<rgba 0 0 1 0>}
    B=Add(A=@n82.RGB, B=@n83.RGB)
    Alpha=NamedRerouteDeclaration(None=@n88) {name=RippleMaskB, nodeColor=<rgba 0 1 0.7891 1>}
   )
  )
  n90 = VectorParameter() {parameterName=NormalStrength, defaultValue=<rgba 2 2 1 0>}
  Reroute(None=Normalize(VectorInput=Multiply(A=@n89, B=@n90.RGB)))
  Saturate(None=Sine(None=Time()) {period=4})
  TextureCoordinate()
  VectorParameter() {group=BaseColor, parameterName=BaseColorPhased, defaultValue=<rgba 0.2774 0.5573 0.3801 0>}
 ]
}
